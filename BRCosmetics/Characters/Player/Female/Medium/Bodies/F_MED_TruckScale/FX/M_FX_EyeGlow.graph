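# Unlit, masked material for the eye glow effect on the F_MED_TruckScale player body, enabled for Niagara mesh particles.
# An eye mask texture supplies a refined red channel plus a green channel scaled by the GlowAmount dynamic parameter.
# That sum is tinted by particle color, added to a dark Background color, then passed through shared emissive adjustments.
# Boosted, saturated blue channel times particle alpha forms a mask, with dithered opacity enabled.
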Material /BRCosmetics/Characters/Player/Female/Medium/Bodies/F_MED_TruckScale/FX/M_FX_EyeGlow
graph {
 properties {
  bAutomaticallySetUsageInEditor=false
  bUsedWithNiagaraMeshParticles=true
  blendMode=BLEND_Masked
  ditherOpacityMask=true
  shadingModel=MSM_Unlit
  usedShadingModels=Unlit
 }
 nodes [
  n1 = TextureSample() {samplerType=SAMPLERTYPE_LinearColor, texture=&/T_FX_EyeMask}
  n2 = VectorParameter() {defaultValue=<rgba 8 8 0.5 0.5>}
  n3 = Add(
   A=MaterialFunctionCall(Texture=@n1.R, RGB=@n2.RGB, A=@n2.RGB) {materialFunction=&MF_FX_Refine}.Result
   B=Multiply(A=@n1.G, B=DynamicParameter() {paramNames [GlowAmount,Param2,Param3,Param4]}.GlowAmount)
  )
  n4 = ParticleColor()
  MaterialFunctionCall(Emissive=Add(A=VectorParameter() {parameterName=Background, defaultValue=<rgba 0.002 0.002 0.002 0>}.RGB, B=Multiply(A=@n3, B=@n4.RGB))) {materialFunction=&MF_FX_EmissiveAdjustments}
  n5 = Constant() {r=3}
  Multiply(A=Saturate(None=Multiply(A=@n1.B, B=@n5)), B=@n4.A)
 ]
}
