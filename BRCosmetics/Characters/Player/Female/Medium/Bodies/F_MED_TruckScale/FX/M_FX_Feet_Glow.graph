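# Additive unlit material for Niagara mesh particles, named as a foot glow effect for the female TruckScale player cosmetic body.
# Particle color RGB passes through the shared FX emissive adjustment function.
# Separate sphere masks on U and V build soft falloffs centered in UV space, shaped by parameterized powers and multiplied, with no textures sampled.
# Particle alpha scales that mask, and an inverted Fresnel fades the mesh's silhouette edges.
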Material /BRCosmetics/Characters/Player/Female/Medium/Bodies/F_MED_TruckScale/FX/M_FX_Feet_Glow
graph {
 properties {
  bAutomaticallySetUsageInEditor=false
  bUsedWithNiagaraMeshParticles=true
  blendMode=BLEND_Additive
  shadingModel=MSM_Unlit
  usedShadingModels=Unlit
 }
 nodes [
  n1 = ParticleColor()
  MaterialFunctionCall(Emissive=@n1.RGB) {materialFunction=&MF_FX_EmissiveAdjustments}
  n2 = MaterialFunctionCall(Float2=TextureCoordinate()) {materialFunction=&BreakOutFloat2Components}
  n3 = Constant() {r=0.5}
  n4 = OneMinus(None=SphereMask(A=@n2.R, B=@n3, Radius=@n3) {hardnessPercent=0})
  n5 = ScalarParameter() {defaultValue=1.5}
  n6 = Constant() {r=0.5}
  n7 = OneMinus(None=SphereMask(A=@n2.G, B=@n6, Radius=@n6) {hardnessPercent=0})
  n8 = ScalarParameter() {defaultValue=5, parameterName=Param_1}
  n9 = Multiply(A=OneMinus(None=Power(Base=@n4, Exp=@n5)), B=OneMinus(None=Power(Base=@n7, Exp=@n8)))
  n10 = ScalarParameter() {defaultValue=1, parameterName=FresnelExponent}
  Multiply(A=Multiply(A=@n1.A, B=Power(Base=@n9)), B=OneMinus(None=Fresnel(ExponentIn=@n10) {baseReflectFraction=0}))
 ]
}
