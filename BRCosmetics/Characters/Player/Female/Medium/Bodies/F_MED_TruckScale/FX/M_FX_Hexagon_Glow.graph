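# Additive unlit material for Niagara mesh particles, providing a hexagon glow effect for the F_MED_TruckScale character body.
# A powered Fresnel, blended between standard and axis-aligned versions by a scalar, is multiplied by a sine mask along V.
# Depth fade and particle alpha shape that mask, while particle RGB passes through a shared emissive adjustment function.
# A scaled world-space vertex normal provides an offset to push the mesh outward.
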Material /BRCosmetics/Characters/Player/Female/Medium/Bodies/F_MED_TruckScale/FX/M_FX_Hexagon_Glow
graph {
 properties {
  bAutomaticallySetUsageInEditor=false
  bUsedWithNiagaraMeshParticles=true
  blendMode=BLEND_Additive
  shadingModel=MSM_Unlit
  usedShadingModels=Unlit
 }
 nodes [
  Multiply(A=VertexNormalWS(), B=ScalarParameter() {defaultValue=2, parameterName=WPO_NormalOffset})
  n1 = MaterialFunctionCall(Exponent=ScalarParameter() {defaultValue=1}) {materialFunction=&MF_Fresnel_FX}
  n2 = MaterialFunctionCall(Axis=VectorParameter() {parameterName=FresnelAxis, defaultValue=<rgba 1 0 0 1>}.RGB) {materialFunction=&AxisAlignedFresnel}
  n3 = ScalarParameter() {defaultValue=1, parameterName=UseFresnelAxis?}
  n4 = Multiply(A=Power(Base=LinearInterpolate(A=@n1.Fresnel, B=@n2.Fresnel, Alpha=@n3), Exp=ScalarParameter() {defaultValue=6, parameterName=FresnelPower})) {constA=1}
  n5 = MaterialFunctionCall(Float2=TextureCoordinate()) {materialFunction=&BreakOutFloat2Components}
  n6 = Abs(None=Sine(None=Multiply(A=@n5.G) {constB=0.5}))
  n7 = ScalarParameter() {defaultValue=2, parameterName=Param_2}
  n8 = Multiply(
   A=Multiply(A=Multiply(B=@n4) {constA=1}, B=Power(Base=@n6, Exp=@n7))
   B=DepthFade(FadeDistance=ScalarParameter() {defaultValue=10, parameterName=FadeDistance})
  )
  n9 = ParticleColor()
  Reroute(None=Multiply(A=Saturate(None=@n8), B=@n9.A))
  Reroute(None=MaterialFunctionCall(Emissive=@n9.RGB) {materialFunction=&MF_FX_EmissiveAdjustments}.Result)
 ]
}
