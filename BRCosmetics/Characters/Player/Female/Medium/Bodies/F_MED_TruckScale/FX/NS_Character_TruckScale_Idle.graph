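# Idle Niagara effect for the TruckScale character body that shows flickering mesh-rendered eyes attached to the skeletal mesh.
# A burst-spawned, non-expiring particle follows a bone or socket with a camera offset.
# An emitter waveform drives Flicker, which modulates a material parameter and brightness, alongside time-of-day dimming.
# User.Variant blends the color, User.Mask sets alpha and spawn group, and bisHeadHidden collapses the mesh to zero scale.
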NiagaraSystem /BRCosmetics/Characters/Player/Female/Medium/Bodies/F_MED_TruckScale/FX/NS_Character_TruckScale_Idle
dependencies {
 usedDataInterfaces [&NiagaraDataInterfaceSkeletalMesh]
 usedDynamicInputs []
 usedModules [
  &self:SystemSpawnScript.SystemScriptSource.SystemScriptGraph.NiagaraNodeAssignment_0.SetVariables_#1
  &SystemState
 ]
}
emitters {Eyes {
  emitterUpdateScript [
   &EmitterState(Inactive Response=::NewEnumerator1, Life Cycle Mode=::NewEnumerator1, Loop Duration=1, UseLoopDelay=false)
   &SpawnBurst_Instantaneous(
    Spawn Group:NiagaraInt32=&MakeIntFromFloat(Float=$User.Mask)
    Use Spawn Probability=true
    UseLoopCountLimit=true
   )
   /Niagara/Noise/Waveform(
    EnableExponent1=false
    GlobalClampMax=false
    GlobalClampMin=false
    UseInput1=false
    UsePhase1=false
    "Waveform [1]":ENiagara_Waveforms=::NewEnumerator12
    "[1] Frequency"=7
   )
   /SetVariables(Emitter.Flicker:NiagaraFloat=&Add_Float(A=&Multiply_Float(A=$Emitter.Waveform.Waveform, B=0.5), B=0.5)) {isSetParameters=true}
  ]
  particleSpawnScript [&InitializeParticle(Lifetime=5, Mesh Scale Mode=::NewEnumerator3, UsePositionOffset=false, Write Material Random=true)]
  particleUpdateScript [
   &ParticleState(Kill Particles When Lifetime Has Elapsed=false)
   &SkeletalMeshLocation(Bone / Socket Index=6, Bone Sampling Mode=::NewEnumerator4, MaskSpawnGroup=false, Skeletal Mesh=$System.Skeletal Mesh)
   &CameraOffset()
   &DynamicMaterialParameters(
    Index 0 Param 1=&Multiply_Float(A=0.25, B=$Emitter.Flicker)
    Module.Index0Float0Write=true
    Module.Index0Float1Write=false
    Module.Index0Float2Write=false
    Module.Index0Float3Write=false
   )
   &ScaleColor(
    Scale Alpha=$User.Mask
    Scale RGB=&Multiply_Vector(
     A=&MakeVectorFromLinearColorRGB(Color=&LerpLinearColors(StartColor=<rgba 3 0.45 2.5795 1>, EndColor=<rgba 2 0.1 0.1 1>, LerpFactor=$User.Variant))
     B=&VectorFromFloat(Value=&Multiply_Float(
       A=&FNTimeOfDayScalar(Night Scale=0.7, Morning Scale=0.8, Evening Scale=0.8)
       B=&Lerp_Float(A=0.7, B=1, Alpha=$Emitter.Flicker)
      ))
    )
    ScaleA=true
    ScaleRGB=true
   )
   &ScaleMeshSize(Scale Factor=&MakeCustomVectorFromBool(Bool=$User.bisHeadHidden, True Vector=<0 0 0>, False Vector=<1 1 1>))
  ]
  rendererClasses [&NiagaraMeshRendererProperties]
  renderers [{rendererClass=&NiagaraMeshRendererProperties, rendererIndex=0}]
 }}
summary {
 systemName=NS_Character_TruckScale_Idle
 emitters [{emitterName=Eyes, rendererClasses [&NiagaraMeshRendererProperties]}]
 userVariables [
  User.bIsFrontEnd {type=NiagaraBool}
  User.bIsFrontEndPreview {type=NiagaraBool}
  User.bisHeadHidden {type=NiagaraBool}
  User.Mask {type=NiagaraFloat, defaultValue=0, description=""}
  User.Reactive {type=NiagaraBool}
  User.Variant {type=NiagaraFloat, defaultValue=0, description=""}
 ]
}
