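# Niagara system for the reactive cosmetic effect on the female TruckScale character body, attaching glows and hexagons to its skeletal mesh.
# FeetFX, Glow and HexagonSoftGlow hold burst-spawned meshes on bones picked by spawn group, while SmallHexagons streams sprites outward from socket-aligned cylinders.
# A shared system AnimationCurve drives intensity, with time-of-day scaling on the glows.
# User.Variant shifts pink to red, and foot/hand-hidden flags mask the matching groups.
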NiagaraSystem /BRCosmetics/Characters/Player/Female/Medium/Bodies/F_MED_TruckScale/FX/NS_Character_TruckScale_Reactivity
dependencies {
 usedDataInterfaces [
  &NiagaraDataInterfaceArrayInt32
  &NiagaraDataInterfaceArrayFloat
  &NiagaraDataInterfaceCurve
  &NiagaraDataInterfaceSkeletalMesh
 ]
 usedDynamicInputs [
  &SetBoolByFloatComparison
  &Abs_Float
  &VectorLength
  /Game/Effects/Niagara/DynamicInputs/State/FN_IsExecutionStateActive
 ]
 usedModules [
  &self:SystemSpawnScript.SystemScriptSource.SystemScriptGraph.NiagaraNodeAssignment_0.SetVariables_#1
  &self:SystemSpawnScript.SystemScriptSource.SystemScriptGraph.NiagaraNodeAssignment_1.SetVariables_#2
  &SystemState
  &Timeline
  &self:SystemSpawnScript.SystemScriptSource.SystemScriptGraph.NiagaraNodeAssignment_2.SetVariables_#3
  &self:SystemSpawnScript.SystemScriptSource.SystemScriptGraph.NiagaraNodeAssignment_3.SetVariables_#4
 ]
}
emitters {
 FeetFX {
  emitterUpdateScript [
   &EmitterState(Inactive Response=::NewEnumerator1, Life Cycle Mode=::NewEnumerator1, Loop Behavior=::NewEnumerator1, Loop Duration=5.5, UseLoopDelay=false)
   &SpawnBurst_Instantaneous(Spawn Count=2, Use Spawn Probability=false, UseLoopCountLimit=true)
  ]
  particleSpawnScript [
   &FN_SetSpawnGroup(Number of Spawn Groups=2)
   &InitializeParticle(Lifetime=5.5, Mesh Scale Mode=::NewEnumerator4, UsePositionOffset=false, Write Material Random=true)
   &CameraOffset(Camera Offset Amount=1)
  ]
  particleUpdateScript [
   /SetVariables(Particles.VisibilityTag=&MakeCustomIntFromBool(Bool=$User.bisFootHidden, True INT=2, False INT=$Particles.SpawnGroup)) {isSetParameters=true, module=SetVariables_#5}
   &ParticleState()
   &SkeletalMeshLocation(
    Bone / Socket Index=&SelectIntFromArray(Direct Array Index=$Particles.SpawnGroup, Int Selection Array=DataInterface() {propertyValues {IntData [0,1]}})
    Bone Sampling Mode=::NewEnumerator4
    MaskSpawnGroup=false
    Sampled Position Offset=<0 0 1.5>
    Sampled Position Offset Coordinate Space:ENiagaraSkelMeshTransforms=::NewEnumerator2
    Skeletal Mesh=$System.Skeletal Mesh
   )
   &ScaleColor(
    Scale Alpha=&MergeHelper_FloatSwitch(0=1, 1=1.5, 2=?, 3=?, Index=&MakeIntFromFloat(Float=$User.Variant), Num=::NewEnumerator0)
    Scale RGB=&Multiply_VectorByFloat(
     Vector=&MakeVectorFromLinearColorRGB(Color=&MergeHelper_ColorSwitch(
       0=<rgba 0.9647 0.1447 0.8295 1>
       1=<rgba 0.4 0 0 1>
       2=?
       3=?
       Index=&MakeIntFromFloat(Float=$User.Variant)
       Num=::NewEnumerator0
      ))
     Float=0.9
    )
    ScaleA=true
    ScaleRGB=true
   )
   &InterpolateOverTime(Rate Of Change=10, Target Vector Value:NiagaraPosition=$Particles.Position, Variable Type=::NewEnumerator2)
   /SetVariables(Particles.Interpolate DY:Vector4f=&MakeFloat4FromVectorAndFloat("Vector (RGB)"=$Particles.InterpolateOverTime.InterpolateOverTime_V3_.Moving Average, "Float (Alpha)"=1)) {isSetParameters=true}
   &DynamicMaterialParameters(
    Index 0 Param 1=1
    Index 0 Param 2=&MakeCustomFloatFromBool(Bool=$System.IsRunning, True Float=0.7, False Float=0.5)
    Index 0 Param 3=1
    Index 0 Param 4=&SelectFloatFromArray(Direct Array Index=$Particles.SpawnGroup, Float Selection Array=DataInterface() {propertyValues {FloatData [0,0.5]}})
    Index 2 Param 1=&FloatFromCurve(
     FloatCurve=DataInterface() {propertyValues {Curve {keys [{interpMode=RCIM_Cubic, tangentMode=RCTM_Break, time=0, value=0.2}, {interpMode=RCIM_Cubic, tangentMode=RCTM_Break, time=1, value=1}]}}}
     CurveIndex=$System.AnimationCurve
    )
    Index 2 Param 2=10
    Index 2 Param 3=1
    Module.Index0Float0Write=true
    Module.Index0Float1Write=true
    Module.Index0Float2Write=true
    Module.Index0Float3Write=true
    Module.Index2Float0Write=true
    Module.Index2Float1Write=true
    Module.Index2Float2Write=true
    Module.Index2Float3Write=true
    Module.Index3Float0Write=true
    Module.Index3Float1Write=true
    Module.Index3Float2Write=true
    Module.Index3Float3Write=true
    Write Parameter Index 2=::NewEnumerator1
    Write Parameter Index 3=::NewEnumerator1
   )
   /SetVariables(Particles.TopColor:LinearColor=&MultiplyLinearColor(
     Linear Color=&MergeHelper_ColorSwitch(
      0=<rgba 1 0.4 0.91 1>
      1=<rgba 3 0.06 0.06 1>
      2=?
      3=?
      Index=&MakeIntFromFloat(Float=$User.Variant)
      Num=::NewEnumerator0
     )
     Scale Factor=&MakeFloat4FromVectorAndFloat(
      "Vector (RGB)"=<1 1 1>
      "Float (Alpha)"=&MakeCustomFloatFromBool(Bool=$User.bisFootHidden, True Float=0, False Float=1)
     )
    )) {isSetParameters=true}
  ]
  rendererClasses [&NiagaraMeshRendererProperties]
  renderers [
   {rendererClass=&NiagaraMeshRendererProperties, rendererIndex=0}
   {rendererClass=&NiagaraMeshRendererProperties, rendererIndex=1}
   {rendererClass=&NiagaraMeshRendererProperties, rendererIndex=2}
   {rendererClass=&NiagaraMeshRendererProperties, rendererIndex=3}
  ]
 }
 Glow {
  emitterUpdateScript [
   &EmitterState(Inactive Response=::NewEnumerator1, Life Cycle Mode=::NewEnumerator1, Loop Behavior=::NewEnumerator1, Loop Duration=5.5, UseLoopDelay=false)
   &SpawnBurst_Instantaneous(Spawn Count=2, Use Spawn Probability=false, UseLoopCountLimit=true)
  ]
  particleSpawnScript [
   &InitializeParticle(Lifetime=5.5, Mesh Scale Mode=::NewEnumerator3, UsePositionOffset=false, Write Material Random=true)
   &FN_SetSpawnGroup(Number of Spawn Groups=2)
   &CameraOffset()
  ]
  particleUpdateScript [
   &ParticleState()
   /SetVariables(Particles.VisibilityTag=&MakeCustomIntFromBool(Bool=$User.bisFootHidden, True INT=2, False INT=$Particles.SpawnGroup)) {isSetParameters=true, module=SetVariables_#5}
   &SkeletalMeshLocation(
    Bone / Socket Index=&SelectIntFromArray(Direct Array Index=$Particles.SpawnGroup, Int Selection Array=DataInterface() {propertyValues {IntData [0,1]}})
    Bone Sampling Mode=::NewEnumerator4
    MaskSpawnGroup=false
    Skeletal Mesh=$System.Skeletal Mesh
   )
   &ScaleColor(
    Scale Alpha=&Multiply_Float(
     A=&Multiply_Float(
      A=$System.AnimationCurve
      B=&FNTimeOfDayScalar(Night Scale=0.5, Morning Scale=1, Evening Scale=1)
     )
     B=&Lerp_Float(A=0.33, B=0.1, Alpha=$User.Variant)
    )
    Scale RGB=&MakeVectorFromLinearColorRGB(Color=&LerpLinearColors(StartColor=<rgba 0.9647 0.1447 0.8295 0.333>, EndColor=<rgba 0.9647 0 0 0.1>, LerpFactor=$User.Variant))
    ScaleA=true
    ScaleRGB=true
   )
  ]
  rendererClasses [&NiagaraMeshRendererProperties]
  renderers [
   {rendererClass=&NiagaraMeshRendererProperties, rendererIndex=0}
   {rendererClass=&NiagaraMeshRendererProperties, rendererIndex=1}
  ]
 }
 HexagonSoftGlow {
  emitterUpdateScript [
   &EmitterState(Inactive Response=::NewEnumerator1, Life Cycle Mode=::NewEnumerator1, Loop Behavior=::NewEnumerator1, Loop Duration=5.6, UseLoopDelay=false)
   &SpawnBurst_Instantaneous(Spawn Count=4, Use Spawn Probability=false, UseLoopCountLimit=true)
  ]
  particleSpawnScript [
   &FN_SetSpawnGroup()
   &InitializeParticle(Lifetime=5.5, Mesh Scale Mode=::NewEnumerator3, UsePositionOffset=false, Write Material Random=true)
   /SetVariables(Particles.VisibilityTag=&SelectIntFromArray(Direct Array Index=$Particles.SpawnGroup, Int Selection Array=DataInterface() {propertyValues {IntData [0,1,2,3]}})) {isSetParameters=true}
  ]
  particleUpdateScript [
   &ParticleState()
   &SkeletalMeshLocation(
    Bone / Socket Index=&SelectIntFromArray(Direct Array Index=$Particles.SpawnGroup, Int Selection Array=DataInterface() {propertyValues {IntData [2,3,4,5]}})
    Bone Sampling Mode=::NewEnumerator4
    MaskSpawnGroup=false
    Skeletal Mesh=$System.Skeletal Mesh
   )
   &ScaleColor(
    Scale Alpha=&Multiply_Float(
     A=&ClampFloat(
      Float=&Multiply_Float(
       A=$System.AnimationCurve
       B=&Multiply_Float(
        A=0.6
        B=&FNTimeOfDayScalar(Night Scale=0.55, Morning Scale=0.7, Day Scale=0.9, Evening Scale=0.6)
       )
      )
      Max=&FN_FrontEndFloatModifier(Gameplay=1, Front End=0.5)
     )
     B=0.5
    )
    Scale RGB=&Lerp_Vector(
     A=&MakeVectorFromLinearColorRGB(Color=<rgba 0.9647 0.1447 0.8295 0.5>)
     B=&MakeVectorFromLinearColorRGB(Color=<rgba 0.9647 0.1447 0.1447 0.5>)
     Alpha=$User.Variant
    )
    ScaleA=true
    ScaleRGB=true
   )
   &ScaleColor(
    Scale Alpha=&MakeCustomFloatFromBool(
     Bool=$User.bisFootHidden
     True Float=&MergeHelper_FloatSwitch(0=0, 1=0, 2=1, 3=1, Index=$Particles.SpawnGroup, Num=::NewEnumerator2)
     False Float=1
    )
    ScaleA=true
    ScaleRGB=false
   ) {module=ScaleColor001}
   &ScaleColor(
    Scale Alpha=&MakeCustomFloatFromBool(
     Bool=$User.bisHandHidden
     True Float=&MergeHelper_FloatSwitch(0=1, 1=1, 2=0, 3=0, Index=$Particles.SpawnGroup, Num=::NewEnumerator2)
     False Float=1
    )
    ScaleA=true
    ScaleRGB=false
   ) {module=ScaleColor002}
  ]
  rendererClasses [&NiagaraMeshRendererProperties]
  renderers [
   {rendererClass=&NiagaraMeshRendererProperties, rendererIndex=0}
   {rendererClass=&NiagaraMeshRendererProperties, rendererIndex=1}
   {rendererClass=&NiagaraMeshRendererProperties, rendererIndex=2}
   {rendererClass=&NiagaraMeshRendererProperties, rendererIndex=3}
  ]
 }
 SmallHexagons {
  emitterUpdateScript [
   &EmitterState(Life Cycle Mode=::NewEnumerator1, Loop Behavior=::NewEnumerator1, UseLoopDelay=false)
   &SpawnRate(SpawnRate=168, Use Spawn Probability=true)
  ]
  particleSpawnScript [
   &InitializeParticle(
    Color=&LerpLinearColors(
     StartColor=<rgba 0.9387 0 0.6399 1>
     EndColor=&RandomRangeLinearColor(Minimum=<rgba 0.9387 0 0.0626 1>, Maximum=<rgba 0.9387 0 0.0156 1>)
     LerpFactor=$User.Variant
    )
    Lifetime=&RandomRangeFloat(Minimum=0.6, Maximum=0.7)
    Sprite Size Mode=::NewEnumerator3
    Uniform Sprite Size=&RandomRangeFloat(Minimum=2.4, Maximum=4.8)
    UsePositionOffset=false
    Write Material Random=true
   )
   &FN_SetSpawnGroup(Number of Spawn Groups=6)
  ]
  particleUpdateScript [
   &ParticleState()
   &SkeletalMeshLocation(
    Bone / Socket Index=&MergeHelper_IntSwitch(0=0, 1=1, 2=2, 3=3, 4=4, 5=5, Index=$Particles.SpawnGroup, Num=::NewEnumerator4)
    MaskSpawnGroup=false
    Mesh Sampling Type=::NewEnumerator5
    Skeletal Mesh=$System.Skeletal Mesh
    Socket Sampling Mode=::NewEnumerator5
    Write To Particle Position=false
   )
   &ShapeLocation(
    Additional Quaternion Rotation=true
    Additional Yaw / Pitch / Roll=&MergeHelper_VectorSwitch(
     0=<0 0.27 -0.04>
     1=<0 0.27 0>
     2=<0 0.27 -0.04>
     3=<0 0.27 0>
     4=<0 0.25 0>
     5=<0 0.25 0>
     Index=$Particles.SpawnGroup
     Num=::NewEnumerator4
    )
    Cylinder Height=&MergeHelper_FloatSwitch(0=17, 1=17, 2=17, 3=17, 4=20, 5=20, Index=$Particles.SpawnGroup, Num=::NewEnumerator4)
    Cylinder Height Midpoint:NiagaraFloat=&MergeHelper_FloatSwitch(0=0, 1=0, 2=1, 3=1, 4=1, 5=0, Index=$Particles.SpawnGroup, Num=::NewEnumerator4)
    Cylinder Radius=&MergeHelper_FloatSwitch(0=4, 1=3, 2=4, 3=3, 4=3, 5=3, Index=$Particles.SpawnGroup, Num=::NewEnumerator4)
    MaskSpawnGroup=false
    Offset=&MergeHelper_VectorSwitch(0=<0 0 0>, 1=<0 0 0>, 2=<0 0 0>, 3=<0 0 0>, 4=<0 0 0>, 5=<0 0 0>, Index=$Particles.SpawnGroup, Num=::NewEnumerator4)
    Offset Coordinate Space=::Local
    Offset Mode=::NewEnumerator0
    Override Seed=false
    Rotation Mode=::NewEnumerator3
    Rotation Quaternion=$StackContext.SkeletalMeshLocation.SampledRotation
    Shape Origin=$Particles.SkeletalMeshLocation.SampledPosition
    Shape Primitive=::NewEnumerator1
    Surface Only=true
    Use Endcaps In Surface Only Mode=false
    UseOwnerScale=true
   )
   /SetVariables(Particles.Position=&Lerp_Vector(
     A=$Particles.ShapeLocation.ShapePosition
     B=&Add_Vector(
      A=$Particles.ShapeLocation.ShapePosition
      B=&Multiply_VectorByFloat(
       Vector=$Particles.ShapeLocation.ShapeNormal
       Float=&Lerp_Float(A=4, B=8, Alpha=&RandomRangeFloat(Minimum=0, Maximum=1))
      )
     )
     Alpha=&FloatFromCurve(
      FloatCurve=DataInterface() {propertyValues {Curve {keys [
          {interpMode=RCIM_Cubic, leaveTangent=1.9983, tangentMode=RCTM_Break, time=0, value=0}
          {arriveTangent=0.1227, interpMode=RCIM_Cubic, tangentMode=RCTM_Break, time=1, value=1}
         ]}}}
      Scale Curve=&Multiply_Float(
       A=&MergeHelper_FloatSwitch(0=1, 1=1, 2=1, 3=1, 4=0.75, 5=0.75, Index=$Particles.SpawnGroup, Num=::NewEnumerator4)
       B=1
      )
     )
    )) {isSetParameters=true}
   &SolveForcesAndVelocity(Clamp Velocity=false, Limit Acceleration=false)
   &SpriteFacingAndAlignment(Sprite Facing=$Particles.ShapeLocation.ShapeNormal, WriteAlignment=false, WriteFacing=true)
   &ScaleColor(
    Scale Alpha=&FloatFromCurve(FloatCurve=DataInterface() {propertyValues {Curve {keys [
         {interpMode=RCIM_Cubic, tangentMode=RCTM_Break, time=0, value=0}
         {interpMode=RCIM_Cubic, tangentMode=RCTM_Break, time=0.25, value=1}
         {arriveTangent=-0.0405, interpMode=RCIM_Cubic, leaveTangent=-0.0405, tangentMode=RCTM_User, time=0.5, value=1}
         {interpMode=RCIM_Cubic, leaveTangent=-8.0089, time=1, value=0}
        ]}}})
    Scale RGB=&VectorFromFloat(Value=&Waveform(
      Global Input Position=&RandomRangeFloat(Minimum=0, Maximum=1)
      "Waveform [1]"=::NewEnumerator0
      "[1] Frequency"=3
      "[1] Amplitude Min/Max"=<0.3 1>
     ))
    ScaleA=true
    ScaleRGB=true
   )
   &ScaleSpriteSize(Uniform Curve Sprite Scale=DataInterface() {propertyValues {Curve {keys [
        {arriveTangent=0.6194, interpMode=RCIM_Cubic, leaveTangent=0.6194, tangentMode=RCTM_User, time=0, value=0.5}
        {arriveTangent=0.2153, interpMode=RCIM_Cubic, leaveTangent=0.2153, tangentMode=RCTM_User, time=1, value=1}
       ]}}})
   &ScaleColor(Scale Alpha=$System.AnimationCurve, ScaleA=true, ScaleRGB=true) {module=ScaleColor001}
   &ScaleColor(
    Scale Alpha=&MakeCustomFloatFromBool(
     Bool=$User.bisFootHidden
     True Float=&MergeHelper_FloatSwitch(0=0, 1=0, 2=0, 3=0, 4=1, 5=1, Index=$Particles.SpawnGroup, Num=::NewEnumerator4)
     False Float=1
    )
    ScaleA=true
    ScaleRGB=false
   ) {module=ScaleColor002}
   &ScaleColor(
    Scale Alpha=&MakeCustomFloatFromBool(
     Bool=$User.bisHandHidden
     True Float=&MergeHelper_FloatSwitch(0=1, 1=1, 2=1, 3=1, 4=0, 5=0, Index=$Particles.SpawnGroup, Num=::NewEnumerator4)
     False Float=1
    )
    ScaleA=true
    ScaleRGB=false
   ) {module=ScaleColor003}
  ]
  rendererClasses [&NiagaraSpriteRendererProperties]
  renderers [{rendererClass=&NiagaraSpriteRendererProperties}]
 }
}
summary {
 systemName=NS_Character_TruckScale_Reactivity
 emitters [
  {emitterName=FeetFX, rendererClasses [&NiagaraMeshRendererProperties]}
  {emitterName=Glow, rendererClasses [&NiagaraMeshRendererProperties]}
  {emitterName=HexagonSoftGlow, rendererClasses [&NiagaraMeshRendererProperties]}
  {emitterName=SmallHexagons, rendererClasses [&NiagaraSpriteRendererProperties]}
 ]
 userVariables [
  User.bisFootHidden {type=NiagaraBool}
  User.bIsFrontEnd {type=NiagaraBool}
  User.bIsFrontEndPreview {type=NiagaraBool}
  User.bisHandHidden {type=NiagaraBool}
  User.DebugReactivity {type=NiagaraBool}
  User.Mask {type=NiagaraFloat, defaultValue=0, description=""}
  User.Reactive {type=NiagaraBool, defaultValue=true}
  User.Variant {type=NiagaraFloat, defaultValue=0, description=""}
 ]
}
