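# Material function for the TruckScale female character body that adds an animated, tinted reactive glow to the incoming material's emissive color.
# Two externally supplied timestamps drive saturated ramps; the second, with an extra time offset, is subtracted from the first and shaped by a power curve.
# That envelope, the tech-art map's blue-channel mask, color and intensity parameters form the glow.
# A debug switch replaces the first ramp with a looping time animation.
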MaterialFunction /BRCosmetics/Characters/Player/Female/Medium/Bodies/F_MED_TruckScale/Materials/MF_TruckScale
graph {nodes [
  n1 = GetMaterialAttributes(None=FunctionInput() {inputType=FunctionInput_MaterialAttributes}) {attributeGetTypes [#1, #2]}
  n2 = NamedRerouteUsage()
  FunctionOutput(None=SetMaterialAttributes!(MaterialAttributes=@n1.MaterialAttributes, EmissiveColor=Add(A=@n1.EmissiveColor, B=@n2), BaseColor=@n1.MaterialAttributes) {attributeSetTypes [#2, #1]})
  n3 = Time()
  n4 = ScalarParameter() {defaultValue=0.5, group=TruckScale, parameterName=DebugTimeSpeed}
  n5 = ScalarParameter() {parameterName=InputTimeStamp}
  n6 = Sign(None=@n5)
  n7 = Divide(A=Subtract(A=Time(), B=@n5), B=ScalarParameter() {defaultValue=1, group=TruckScale, parameterName=AnimationDuration})
  n8 = StaticSwitchParameter(True=Frac(None=Multiply(A=@n3, B=@n4)), False=Saturate(None=Multiply(A=@n6, B=@n7))) {parameterName=DebugAnimation}
  n9 = ScalarParameter() {parameterName=InputTimeStampB}
  n10 = ScalarParameter() {defaultValue=0.4, parameterName=TimeOffsetAmount}
  n11 = NamedRerouteUsage()
  n12 = Subtract(A=Subtract(A=Time(), B=Multiply(A=@n10, B=@n11)), B=@n9)
  n13 = ScalarParameter() {defaultValue=1, group=TruckScale, parameterName=AnimationDurationB}
  n14 = Saturate(None=Multiply(A=Sign(None=@n9), B=Divide(A=@n12, B=@n13)))
  NamedRerouteDeclaration(None=Power(Base=Subtract(A=@n8, B=@n14), Exp=ScalarParameter() {defaultValue=0.5, group=TruckScale, parameterName=AnimationInputPowerEXP})) {name=Animation0-1, nodeColor=<rgba 1 0.4453 0 1>}
  NamedRerouteDeclaration(None=ComponentMask(None=FunctionInput() {inputName=TechArtMapInput}) {b=true}) {name=GlowMask, nodeColor=<rgba 0 0.2031 1 1>}
  n15 = NamedRerouteUsage()
  n16 = NamedRerouteUsage()
  n17 = VectorParameter() {parameterName=ReactivityColor, defaultValue=<rgba 0.25 0.0267 0.1074 1>}
  n18 = ScalarParameter() {defaultValue=8, parameterName=ReactiveEmissiveMult}
  NamedRerouteDeclaration(None=Multiply(A=Multiply(A=@n15, B=@n16), B=Multiply(A=@n17.RGB, B=@n18))) {name=Emissive, nodeColor=<rgba 0 0.7422 1 1>}
  Sine(None=@n8)
 ]}
