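# Additive, two-sided skeletal mesh material for the TruckScale character's hexagon effect, with separate leg and arm controls.
# Vertex-color green seeds sine-hash randoms that vary each hex's hue, opacity and idle shrink.
# Timestamp parameters drive in/out animation, while world position offset pushes along normals, shrinks hexes or collapses them to a point.
# Pre-skinned height masks the legs, alongside fresnel emissive, rotating edge tracing and optional vertex-color hiding.
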Material /BRCosmetics/Characters/Player/Female/Medium/Bodies/F_MED_TruckScale/Materials/M_TruckScale_Hex
graph {
 properties {
  bAutomaticallySetUsageInEditor=false
  bUsedWithMeshDeformer=true
  bUsedWithMorphTargets=true
  bUsedWithSkeletalMesh=true
  blendMode=BLEND_Additive
  twoSided=true
  usedShadingModels=Default Lit
 }
 nodes [
  Constant()
  Constant() {r=1}
  Constant() {r=1}
  Constant()
  n1 = Time()
  n2 = Multiply(A=ScalarParameter() {defaultValue=0.4, parameterName=TimeOffsetAmount}, B=NamedRerouteUsage())
  Frac(None=Multiply(A=Subtract(A=@n1, B=@n2), B=ScalarParameter() {defaultValue=0.5, group=TruckScale, parameterName=DebugTimeSpeed}))
  n3 = Min(A=NamedRerouteUsage(), B=NamedRerouteUsage())
  Multiply(A=@n3, B=MaterialFunctionCall() {materialFunction=&DebugTimeSine}.Result)
  n4 = TextureCoordinate()
  n5 = VectorParameter() {parameterName=HexPieceUVCropping, defaultValue=<rgba 0.625 0.014 0.69 0.08>}
  n6 = AppendVector(A=@n5.R, B=@n5.R)
  n7 = AppendVector(A=@n5.B, B=@n5.B)
  NamedRerouteDeclaration(None=Saturate(None=MaterialFunctionCall!(UVs=@n4, TextureIn, UpperLeftCorner=@n6, LowerRightCorner=@n7, ExportFloat4) {materialFunction=/Engine/Functions/Engine_MaterialFunctions01/Texturing/TextureCropping}.Crop UVs)) {name=AdjustedHexUV, nodeColor=<rgba 0 0.0859 1 1>}
  n8 = ScalarParameter() {parameterName=InputTimeStamp}
  n9 = Sign(None=@n8)
  n10 = Time()
  n11 = Divide(A=Subtract(A=Subtract(A=@n10, B=@n2), B=@n8), B=ScalarParameter() {defaultValue=1, group=TruckScale, parameterName=AnimationDuration})
  n12 = StaticSwitchParameter(True=Constant() {r=0.2}, False=Saturate(None=Multiply(A=@n9, B=@n11))) {parameterName=DebugAnimation}
  n13 = ScalarParameter() {parameterName=InputTimeStampB}
  n14 = ScalarParameter() {defaultValue=0.4, parameterName=TimeOffsetAmount}
  n15 = NamedRerouteUsage()
  n16 = Subtract(A=Subtract(A=Time(), B=Multiply(A=@n14, B=@n15)), B=@n13)
  n17 = ScalarParameter() {defaultValue=1, group=TruckScale, parameterName=AnimationDurationB}
  n18 = Saturate(None=Multiply(A=Sign(None=@n13), B=Divide(A=@n16, B=@n17)))
  n19 = Subtract(A=@n12, B=@n18)
  n20 = ScalarParameter() {defaultValue=50, group=TruckScale, parameterName=AnimationInputPowerEXP}
  n21 = Multiply(A=NamedRerouteUsage()) {constB=5}
  NamedRerouteDeclaration(None=Power(Base=OneMinus(None=@n19), Exp=Add(A=@n20, B=@n21))) {name=AnimeInput, nodeColor=<rgba 0.5703 1 0 1>}
  n22 = NamedRerouteUsage()
  n23 = Multiply(
   A=VertexNormalWS()
   B=Multiply(A=ConstantBiasScale(None=@n22) {bias=-0.5, scale=2}, B=ScalarParameter() {defaultValue=1, group=TruckScale, parameterName=WPO_VNPushAmount})
  )
  n24 = Multiply(
   A=LinearInterpolate(
    A=ScalarParameter() {defaultValue=-1, group=TruckScale, parameterName=LegArmorPush_Low}
    B=ScalarParameter() {defaultValue=-10, group=TruckScale, parameterName=LegArmorPush_High}
    Alpha=NamedRerouteUsage()
   )
   B=VertexNormalWS()
  )
  n25 = LinearInterpolate(
   A=ScalarParameter() {group=TruckScale, parameterName=LegGradientPush_Low}
   B=ScalarParameter() {defaultValue=2, group=TruckScale, parameterName=LegGradientPush_High}
   Alpha=NamedRerouteUsage()
  )
  n26 = VertexNormalWS()
  NamedRerouteDeclaration(None=Add(A=Add(A=@n23, B=@n24), B=Multiply(A=@n25, B=@n26))) {name=VNPush, nodeColor=<rgba 0 0.3906 1 1>}
  n27 = LocalPosition() {localOrigin=InstancePreSkinning}
  NamedRerouteDeclaration(None=ShaderStageSwitch(PixelShader=VertexInterpolator(VS=@n27.Z).PS, VertexShader=@n27.Z)) {name=PreskinnedHeight, nodeColor=<rgba 1 0.9609 0 1>}
  n28 = NamedRerouteUsage()
  n29 = Step(Y=@n28) {constX=60}
  n30 = MaterialFunctionCall(Input=@n28, Input Low=Constant(), Input High=Constant() {r=160}, Target Low=Constant() {r=-0.55}, Target High=Constant() {r=1.5}) {materialFunction=&RemapValueRange}
  NamedRerouteDeclaration(None=Saturate(None=Multiply(A=@n29, B=@n30.Result))) {name=LegArmorMask, nodeColor=<rgba 0.6406 0 1 1>}
  n31 = MaterialFunctionCall(UVs=NamedRerouteUsage(), Texture Scale=Constant() {r=0.5}) {materialFunction=&ScaleUVsByCenter}
  n32 = TextureSample(UVs=@n31.UVs) {
   samplerType=SAMPLERTYPE_LinearColor
   texture=/BRCosmetics/Characters/Player/Female/Medium/Bodies/F_MED_TruckScale/Textures/T_Holotable_Hexagons
  }
  n33 = Frac(None=Time())
  n34 = Multiply(
   A=Multiply(A=@n32.R, B=@n31.0-1 mask)
   B=MaterialFunctionCall(Vector or UVs=MaterialFunctionCall(UVs=@n31.UVs, "Rotation Angle (0-1)"=@n33) {materialFunction=&CustomRotator}.Rotated Values) {materialFunction=&VectorToRadialValue}.Vector Converted to Angle
  )
  n35 = Step(Y=NamedRerouteUsage(), X=ScalarParameter() {defaultValue=1, parameterName=EdgeTracingThreshold})
  NamedRerouteDeclaration(None=Multiply(A=Multiply(A=@n34, B=@n35), B=ScalarParameter() {defaultValue=100, parameterName=EdgeTracingEmissiveMult})) {name=EdgeTracing, nodeColor=<rgba 1 0.8438 0 1>}
  n36 = ConstantBiasScale(None=NamedRerouteUsage()) {bias=-0.5, scale=2}
  n37 = Frac(None=Multiply(A=Time()) {constB=0.4})
  n38 = SphereMask(A=Add(A=NamedRerouteUsage(), B=Multiply(A=@n36) {constB=5}), B=Add(A=Constant() {r=15}, B=Multiply(A=@n37) {constB=180})) {attenuationRadius=10, hardnessPercent=0}
  n39 = NamedRerouteUsage()
  NamedRerouteDeclaration(None=Power(Base=Multiply(A=@n38, B=@n39))) {name=LegGradientMask, nodeColor=<rgba 0.3828 1 0 1>}
  n40 = Step(Y=@n3) {constX=0.5}
  n41 = Constant() {r=0.8}
  n42 = Multiply(
   A=VectorParameter() {group=TruckScale, parameterName=LegArmorVertical, defaultValue=<rgba 0 0 1.5 1>}.RGB
   B=ConstantBiasScale(None=MaterialFunctionCall(Speed=@n41) {materialFunction=&DebugTimeSine}.Result) {bias=0.5, scale=1}
  )
  NamedRerouteDeclaration(None=Multiply(A=Multiply(A=@n40, B=@n29), B=Transform(None=@n42) {transformSourceType=TRANSFORMSOURCE_Local})) {name=LegArmorVerticalFloating, nodeColor=<rgba 1 0.4219 0 1>}
  NamedRerouteDeclaration(None=@n18) {name=FadeOut, nodeColor=<rgba 1 0 0.7266 1>}
  NamedRerouteDeclaration(None=@n29) {name=LegMask, nodeColor=<rgba 1 0 0.7266 1>}
  n43 = TransformPosition(None=Constant3Vector() {constant=<rgba 0 0 110 0>}) {transformType=TRANSFORMPOSSOURCE_World}
  n44 = WorldPosition() {worldPositionShaderOffset=WPT_ExcludeAllShaderOffsets}
  n45 = NamedRerouteUsage()
  n46 = MaterialFunctionCall(
   VectorToTransform=AppendVector(A=ConstantBiasScale(None=@n45) {bias=-0.5, scale=2}, B=Constant())
   BasisX=Transform(None=Constant3Vector() {constant=<rgba 1 0 0 0>})
   BasisY=Transform(None=Constant3Vector() {constant=<rgba 0 1 0 0>})
   BasisZ=Transform(None=Constant3Vector() {constant=<rgba 0 0 1 0>})
   BasisPosition=Constant3Vector()
  ) {materialFunction=&Transform3x3Matrix}
  n47 = Multiply(A=Multiply(A=@n46.Transformed Vector, B=ScalarParameter() {defaultValue=-4, parameterName=HexShrinkAmount}), B=NamedRerouteUsage()) {constB=-1}
  n48 = NamedRerouteUsage()
  n49 = Add(A=Add(A=Add(A=@n47, B=@n48), B=NamedRerouteUsage()), B=NamedRerouteUsage())
  n50 = NamedRerouteUsage()
  n51 = NamedRerouteUsage()
  n52 = ScalarParameter() {defaultValue=0.77, parameterName=Threshold}
  NamedRerouteDeclaration(None=LinearInterpolate(A=Subtract(A=@n43, B=@n44.XYZ), B=Add(A=@n49, B=@n50), Alpha=Step(Y=@n51, X=@n52))) {name=WPO, nodeColor=<rgba 1 0 0.8906 1>}
  n53 = NamedRerouteUsage()
  n54 = MaterialFunctionCall(
   Hue Shift Percentage=Multiply(A=ConstantBiasScale(None=@n53) {bias=-0.5, scale=2}, B=ScalarParameter() {defaultValue=-0.1, parameterName=HueShiftRandomAmount})
   Texture=VectorParameter() {parameterName=HexColor, defaultValue=<rgba 2 0 1.9 1>}.RGB
  ) {materialFunction=&HueShift}
  n55 = VertexNormalWS()
  n56 = ScalarParameter() {defaultValue=1, parameterName=FresnelExp}
  NamedRerouteDeclaration(None=LinearInterpolate(
    A=Multiply(A=@n54.Result) {constB=0.2}
    B=@n54.Result
    Alpha=MaterialFunctionCall(Normal Vector=@n55, Power=@n56) {materialFunction=&Fresnel_Function}.Result
   )) {name=HexEmissive, nodeColor=<rgba 0 1 0.7891 1>}
  n57 = ScalarParameter() {defaultValue=0.5, parameterName=Hex_Opacity_Max}
  n58 = ScalarParameter() {defaultValue=0.3, parameterName=Hex_Opacity_Min}
  n59 = SphereMask(A=NamedRerouteUsage(), B=Constant() {r=0.5}) {attenuationRadius=0.6, hardnessPercent=0}
  n60 = NamedRerouteUsage()
  n61 = NamedRerouteUsage()
  n62 = ConstantBiasScale(None=LinearInterpolate(A=@n60, B=Min(A=@n60, B=@n61), Alpha=NamedRerouteUsage())) {bias=-0.5, scale=2}
  n63 = ScalarParameter() {defaultValue=1.3, parameterName=OpacitryRandomAmount}
  n64 = Multiply(A=Add(A=LinearInterpolate(A=@n57, B=@n58, Alpha=@n59), B=NamedRerouteUsage()), B=Add(B=Multiply(A=@n62, B=@n63)) {constA=1})
  n65 = Multiply(A=@n64, B=ScalarParameter() {defaultValue=2, group=TruckScale, parameterName=ArmOpacityBoost})
  n66 = NamedRerouteUsage()
  n67 = NamedRerouteUsage()
  n68 = Add(
   A=Multiply(A=NamedRerouteUsage(), B=ScalarParameter() {group=TruckScale, parameterName=HideLegFX})
   B=Multiply(A=OneMinus(None=@n67), B=ScalarParameter() {group=TruckScale, parameterName=HideArmFX})
  )
  NamedRerouteDeclaration(None=Multiply(A=LinearInterpolate(A=@n65, B=@n64, Alpha=@n66), B=LinearInterpolate(Alpha=@n68) {constA=1, constB=0})) {name=HexOpacity, nodeColor=<rgba 1 0 0.0937 1>}
  n69 = Sine(None=Add(A=NamedRerouteUsage(), B=ScalarParameter() {defaultValue=1234.6547, desc=Seed, parameterName=SeedC}))
  n70 = Constant() {r=10000}
  NamedRerouteDeclaration(None=Frac(None=Multiply(A=@n69, B=@n70))) {name=RandomC, nodeColor=<rgba 0.5469 0 1 1>}
  n71 = LinearInterpolate(
   A=ScalarParameter() {defaultValue=-4, parameterName=HexShrinkIdle}
   B=ScalarParameter() {defaultValue=-3, parameterName=HexShrinkIdle_B}
   Alpha=NamedRerouteUsage()
  )
  n72 = LinearInterpolate(
   A=NamedRerouteUsage()
   B=NamedRerouteUsage()
   Alpha=MaterialFunctionCall(Speed=Constant() {r=0.4}) {materialFunction=&DebugTimeSine}.Result
  )
  n73 = LinearInterpolate(
   A=NamedRerouteUsage()
   B=NamedRerouteUsage()
   Alpha=MaterialFunctionCall(Speed=Constant() {r=0.7}) {materialFunction=&DebugTimeSine}.Result
  )
  n74 = MaterialFunctionCall(Speed=Constant() {r=0.5}) {materialFunction=&DebugTimeSine}
  n75 = Multiply(
   A=ConstantBiasScale(None=LinearInterpolate(A=@n72, B=@n73, Alpha=@n74.Result))
   B=ScalarParameter() {defaultValue=5, parameterName=HexShrinkIdleRandomAmount}
  )
  NamedRerouteDeclaration(None=Multiply(A=@n46.Transformed Vector, B=Add(A=@n71, B=@n75))) {name=IdleShrink, nodeColor=<rgba 0 0.0391 1 1>}
  n76 = VertexColor()
  n77 = Sine(None=Add(A=NamedRerouteDeclaration(None=@n76.G) {name=HexVertColorG, nodeColor=<rgba 0.2422 1 0 1>}, B=Constant() {desc=Seed, r=11.11}))
  n78 = Constant() {r=10000}
  NamedRerouteDeclaration(None=Frac(None=Multiply(A=@n77, B=@n78))) {name=RandomA, nodeColor=<rgba 1 0.8672 0 1>}
  n79 = Sine(None=Add(A=NamedRerouteUsage(), B=Constant() {desc=Seed, r=0.214}))
  n80 = Constant() {r=10000}
  NamedRerouteDeclaration(None=Frac(None=Multiply(A=@n79, B=@n80))) {name=RandomB, nodeColor=<rgba 0.5469 0 1 1>}
  NamedRerouteUsage()
  NamedRerouteUsage()
  NamedRerouteUsage()
  NamedRerouteUsage()
  OneMinus(None=@n18)
  Sine(None=@n12)
  n81 = NamedRerouteUsage()
  n82 = MaterialFunctionCall!(WPO=@n81) {materialFunction=/Game/Packages/Fortress_SharedMaterials/Base_Material_Functions/Fortnite_HideElementsWithVertexColor_2022}
  StaticSwitchParameter(
   True=StaticSwitchParameter(True=LinearInterpolate(A=@n81, B=@n82.Result) {constAlpha=0}, False=@n82.Result) {group=Opacity Mask, parameterName=OnlyIncludeVertexColorHideForHolo}
   False=@n81
  ) {group=03_Opacity Mask, parameterName=Use Vertex Colors for Mask}
 ]
}
