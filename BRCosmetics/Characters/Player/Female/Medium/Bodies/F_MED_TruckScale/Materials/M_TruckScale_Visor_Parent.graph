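# Translucent, lit parent material for the TruckScale character's glass visor, usable on skeletal meshes, with diffuse, normal and emissive textures.
# Opacity combines a Fresnel gradient with a scene-versus-pixel depth mask for what lies underneath.
# An edge-enhancement mask follows sky light direction and time of day, alongside a refracted skylight environment sample.
# An eye-area mask color-keys the scene's diffuse buffer against a sampled color parameter.
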Material /BRCosmetics/Characters/Player/Female/Medium/Bodies/F_MED_TruckScale/Materials/M_TruckScale_Visor_Parent
graph {
 properties {
  bAutomaticallySetUsageInEditor=false
  bUsedWithMeshDeformer=true
  bUsedWithSkeletalMesh=true
  blendMode=BLEND_Translucent
  translucencyLightingMode=TLM_SurfacePerPixelLighting
  usedShadingModels=Default Lit
 }
 nodes [
  n1 = TextureSampleParameter2D() {parameterName=Emissive, texture=/BRCosmetics/Characters/Player/Female/Medium/Bodies/F_MED_TruckScale/Textures/T_TruckScale_FaceAccExtra_E}
  Add(A=@n1.RGB, B=NamedRerouteUsage())
  Constant()
  n2 = Multiply(
   A=Fresnel(ExponentIn=ScalarParameter() {defaultValue=5, parameterName=Fresnel Exponent})
   B=ScalarParameter() {defaultValue=1, parameterName=Glass thickness}
  )
  n3 = Power(Base=LinearInterpolate(
    A=ScalarParameter() {defaultValue=1, parameterName=Fresnel Inner Transparency}
    B=ScalarParameter() {parameterName=Fresnel Outer Transparency}
    Alpha=Clamp(None=@n2)
   )) {constExponent=3}
  n4 = NamedRerouteUsage()
  n5 = ComponentMask(None=SceneDepth().R) {b=true, g=true, r=true}
  n6 = PixelDepth()
  n7 = Step(
   Y=Divide(A=Subtract(A=@n5, B=@n6.R) {constB=50}, B=ScalarParameter() {defaultValue=9, parameterName=Division}) {constB=50}
   X=ScalarParameter() {defaultValue=0.999, parameterName=UnderneathMaskThreshold}
  ) {constX=0.1}
  NamedRerouteDeclaration(None=LinearInterpolate(A=Constant() {r=1}, B=Subtract(A=@n3, B=@n4), Alpha=Saturate(None=@n7))) {name=OpacityWithDepth, nodeColor=<rgba 0 1 0.7891 1>}
  n8 = TextureSampleParameter2D() {
   parameterName=Normal
   samplerType=SAMPLERTYPE_Normal
   texture=/BRCosmetics/Characters/Player/Female/Medium/Bodies/F_MED_TruckScale/Textures/T_TruckScale_FaceAccExtra_N
  }
  NamedRerouteDeclaration(None=@n8.RGB) {name=Normal, nodeColor=<rgba 0.4297 1 0 1>}
  n9 = ScalarParameter() {defaultValue=1, parameterName=Fresnel Inner Transparency2}
  n10 = ScalarParameter() {parameterName=Fresnel Outer Transparency2}
  n11 = ScalarParameter() {defaultValue=100, parameterName=Fresnel Exponent2}
  n12 = PixelNormalWS()
  n13 = Clamp(None=Multiply(A=Fresnel(ExponentIn=@n11, Normal=@n12), B=ScalarParameter() {defaultValue=1, parameterName=Glass thickness2}))
  NamedRerouteDeclaration(None=MaterialFunctionCall(In=LinearInterpolate(A=@n9, B=@n10, Alpha=@n13), Contrast=ScalarParameter() {defaultValue=1, parameterName=Param_1}) {materialFunction=&CheapContrast}.Result) {name=FresnelOuter, nodeColor=<rgba 1 0.6094 0 1>}
  n14 = TextureSampleParameter2D() {parameterName=Diffuse, texture=/BRCosmetics/Characters/Player/Female/Medium/Bodies/F_MED_TruckScale/Textures/T_TruckScale_FaceAccExtra_D}
  NamedRerouteDeclaration(None=@n14.RGB) {name=Diffuse, nodeColor=<rgba 0.5234 0 1 1>}
  n15 = DotProduct(A=SkyAtmosphereLightDirection(), B=ComponentMask(None=PixelNormalWS()) {b=true})
  n16 = Saturate(None=Abs(None=@n15))
  n17 = Saturate(None=Sign(None=@n15))
  n18 = LinearInterpolate(
   A=ScalarParameter() {defaultValue=5, parameterName=EnvLightPow_Low}
   B=ScalarParameter() {defaultValue=2, parameterName=EnvLightPow_High}
   Alpha=@n17
  ) {constA=0.3}
  n19 = Multiply(
   A=MaterialFunctionCall(In=Power(Base=@n16, Exp=@n18), Contrast=ScalarParameter() {parameterName=Param_3}) {materialFunction=&CheapContrast}.Result
   B=NamedRerouteUsage()
  )
  n20 = LinearInterpolate(
   A=ScalarParameter() {defaultValue=0.3, parameterName=EnvLight_Low}
   B=ScalarParameter() {defaultValue=3, parameterName=EnvLight_High}
   Alpha=@n17
  ) {constA=0.3}
  n21 = Constant() {r=0.3}
  n22 = Constant() {r=0.5}
  NamedRerouteDeclaration(None=Multiply(
    A=Multiply(A=@n19, B=@n20)
    B=MaterialFunctionCall(Value_Morning=@n21, Value_Day=@n22, Value_Evening=@n22, Value_Night=@n21) {materialFunction=&Scalar_Time_of_Day}.Result
   )) {name=EdgeEnhancementMask, nodeColor=<rgba 1 0.9844 0 1>}
  n23 = Multiply(A=CameraVectorWS()) {constB=-1}
  n24 = PixelNormalWS()
  n25 = Constant() {r=2}
  NamedRerouteDeclaration(None=SkyLightEnvMapSample(Direction=MaterialFunctionCall!(Ray Direction=@n23, Surface Normal=@n24, Refractive Index Origin, Refractive Index Target=@n25) {materialFunction=/Engine/Functions/Engine_MaterialFunctions01/Vectors/Refract}.Vector)) {name=EnvironmentMap, nodeColor=<rgba 1 0 0.5391 1>}
  n26 = ComponentMask(None=SceneTexture() {sceneTextureId=PPI_DiffuseColor}.Color) {b=true, g=true, r=true}
  n27 = VectorParameter() {parameterName=EyeAreaDiffuseSample, defaultValue=<rgba 0.5781 0.5209 0.5766 1>}
  n28 = ScalarParameter() {defaultValue=0.999}
  NamedRerouteDeclaration(None=Power(Base=SphereMask(A=@n26, B=@n27.RGB, Radius=@n28) {attenuationRadius=1, hardnessPercent=0})) {name=EyeArea, nodeColor=<rgba 1 0 0.375 1>}
  NamedRerouteUsage()
  NamedRerouteUsage()
  OneMinus(None=@n16)
  ScalarParameter() {defaultValue=0.35, parameterName=Roughness}
 ]
}
