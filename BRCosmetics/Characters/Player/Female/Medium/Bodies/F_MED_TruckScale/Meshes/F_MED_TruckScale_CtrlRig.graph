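# Control Rig for the F_MED_TruckScale female body cosmetic, adding corrective deformation to pads, armor and body joints.
# After a shared DeformRig function, a sequence uses spherical pose readers to weight bone offsets for shoulders, armor, heels, elbows and knees.
# Shoulder pads also react to upper-arm translation from positions cached at setup.
# Thigh readers, parented to aim-constrained bones, drive the middle glute by maximum; unconnected posing nodes remain.
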ControlRigBlueprint /BRCosmetics/Characters/Player/Female/Medium/Bodies/F_MED_TruckScale/Meshes/F_MED_TruckScale_CtrlRig
graphs [
 RigVMFunctionLibrary {links [], nodes []}
 RigVMFunction_Sequence_1_ContainedGraph {nodes [
   n1 = RigVMFunctionEntryNode()
   RigVMFunction_Sequence = &RigVMFunction_Sequence(ExecuteContext=@n1.ExecuteContext)
   n2 = &RigVMFunction_Sequence(ExecuteContext=@RigVMFunction_Sequence.B)
   n3 = &RigVMFunction_Sequence(ExecuteContext=@n2.B)
   n4 = &RigVMFunction_Sequence(ExecuteContext=@n3.B)
   n5 = &RigVMFunction_Sequence(ExecuteContext=@n4.B)
   RigVMFunctionReturnNode(A=@RigVMFunction_Sequence.A, B=@n2.A, C=@n3.A, D=@n4.A, E=@n5.A, F=@n5.B)
  ]}
 RigVMModel {nodes [
   n1 = &RigUnit_PrepareForExecution()
   RigVMVariableNode(
    ExecuteContext=@n1.ExecutePin
    Value=&RigUnit_GetTransform(Item=(Type=Bone,Name="upperarm_l"), Space=LocalSpace, bInitial=False).Transform.Translation
   ) {variable=Upperarm_l_pos_ref}
   > RigVMVariableNode(Value=&RigUnit_GetTransform(Item=(Type=Bone,Name="upperarm_r"), Space=LocalSpace, bInitial=False).Transform.Translation) {variable=Upperarm_r_pos_ref}
   RigUnit_BeginExecution = &RigUnit_BeginExecution()
   FN_DeformRig_FNC = RigVMFunctionReferenceNode(
    DeformRig_ Curve="(deformRig_l_upperarm_up_alpha_72_#1=False,deformRig_r_upperarm_up_alpha_73_#2=False,deformRig_l_upperarm_fr_alpha_74_#3=False,deformRig_r_upperarm_fr_alpha_75_CB51B8 ...(truncated)"
    DeformRig_Bool="(upperarm_1_#4=true,elbow_3_#5=true,wrist_7_#6=true,groin_8_#7=true,glute_17_#8=true,knee_9_40C408324CA261 ...(truncated)"
    DeformRig_Debug=(upperarm_debug_18_#4=False,wrist_debug_20_#6=False,lat_debug_26_#9=False,groinGlute_debug_27_#7=False)
    DeformRig_Value="(upperarm_up_18_#10=(Rotation=(X=0,Y=0,Z=0,W=1),Translation=(X=0,Y=0,Z=8),Scale3D=(X=0.7,Y=1,Z=1)),upperarm_fwd_fr_19_#11=(Rotation=(X=0,Y=0,Z=0,W=1),Translation=(X=0,Y=10,Z=0),Scale3D=(X=0.5,Y=1,Z= ...(truncated)"
    ExecuteContext=@RigUnit_BeginExecution.ExecutePin
   )
   > RigVMFunction_Sequence = RigVMAggregateNode()
   n2 = &RigUnit_SphericalPoseReader(
    ActiveRegionScaleFactors=(PositiveWidth=1,NegativeWidth=1,PositiveHeight=1,NegativeHeight=1)
    ActiveRegionSize=0.25
    Debug=(bDrawDebug=false,bDraw2D=False,bDrawLocalAxes=False,DebugScale=15,DebugSegments=10,DebugThickness=0.1)
    DriverAxis=(X=-1,Y=0,Z=0)
    DriverItem=(Type=Bone,Name="upperarm_r")
    FalloffRegionScaleFactors=(PositiveWidth=1,NegativeWidth=1,PositiveHeight=1,NegativeHeight=1)
    FalloffSize=0.5
    FlipHeightScaling=False
    FlipWidthScaling=False
    OptionalParentItem=(Type=Bone,Name="None")
    RotationOffset=(X=180,Y=0,Z=0)
    ExecutePin=@RigVMFunction_Sequence.A
   )
   > &RigUnit_OffsetTransformForItem(
    Item=(Type=Bone,Name="pad_r")
    OffsetTransform=(Rotation=(X=0,Y=0.1305,Z=0,W=0.9914),Translation=(X=0,Y=0,Z=4),Scale3D=(X=1,Y=1,Z=1))
    bPropagateToChildren=True
    Weight=@n2.OutputParam
   )
   RigVMFunction_MathVectorDistance = &RigVMFunction_MathVectorDistance(
    B=&RigUnit_GetTransform(Item=(Type=Bone,Name="upperarm_r"), Space=LocalSpace, bInitial=False).Transform.Translation
    A=RigVMVariableNode() {variable=Upperarm_r_pos_ref}.Value
   )
   Remap = &RigVMFunction_MathDoubleRemap(SourceMaximum=2, SourceMinimum=0, TargetMaximum=1, TargetMinimum=0, bClamp=false, Value=@RigVMFunction_MathVectorDistance.Result)
   > &RigUnit_OffsetTransformForItem(
    Item=(Type=Bone,Name="pad_r")
    OffsetTransform=(Rotation=(X=0,Y=0.061,Z=0,W=0.9981),Translation=(X=-0.8,Y=0,Z=0),Scale3D=(X=1,Y=1.1,Z=1))
    bPropagateToChildren=True
    Weight=@Remap.Result
   )
   > n3 = &RigUnit_SphericalPoseReader(
    ActiveRegionScaleFactors=(PositiveWidth=1,NegativeWidth=2.5,PositiveHeight=2.5,NegativeHeight=1)
    ActiveRegionSize=0.15
    Debug=(bDrawDebug=false,bDraw2D=False,bDrawLocalAxes=False,DebugScale=15,DebugSegments=10,DebugThickness=0.1)
    DriverAxis=(X=-1,Y=0,Z=0)
    DriverItem=(Type=Bone,Name="upperarm_r")
    FalloffRegionScaleFactors=(PositiveWidth=1,NegativeWidth=3,PositiveHeight=1,NegativeHeight=2)
    FalloffSize=0.5
    FlipHeightScaling=False
    FlipWidthScaling=False
    OptionalParentItem=(Type=Bone,Name="spine_05")
    RotationOffset=(X=-90,Y=0,Z=40)
   )
   > &RigUnit_OffsetTransformForItem(
    Item=(Type=Bone,Name="pad_r")
    OffsetTransform=(Rotation=(X=0,Y=0,Z=-0.0436,W=0.999),Translation=(X=-0.5,Y=1,Z=0),Scale3D=(X=1,Y=1,Z=1))
    bPropagateToChildren=True
    Weight=@n3.OutputParam
   )
   > n4 = &RigUnit_SphericalPoseReader(
    ActiveRegionScaleFactors=(PositiveWidth=1,NegativeWidth=1,PositiveHeight=1,NegativeHeight=1)
    ActiveRegionSize=0.25
    Debug=(bDrawDebug=false,bDraw2D=False,bDrawLocalAxes=False,DebugScale=15,DebugSegments=10,DebugThickness=0.1)
    DriverAxis=(X=1,Y=0,Z=0)
    DriverItem=(Type=Bone,Name="upperarm_l")
    FalloffRegionScaleFactors=(PositiveWidth=1,NegativeWidth=1,PositiveHeight=1,NegativeHeight=1)
    FalloffSize=0.5
    FlipHeightScaling=False
    FlipWidthScaling=False
    OptionalParentItem=(Type=Bone,Name="None")
    RotationOffset=(X=0,Y=0,Z=0)
   )
   > &RigUnit_OffsetTransformForItem(
    Item=(Type=Bone,Name="pad_l")
    OffsetTransform=(Rotation=(X=0,Y=-0.1305,Z=0,W=0.9914),Translation=(X=0,Y=0,Z=-4),Scale3D=(X=1,Y=1,Z=1))
    bPropagateToChildren=True
    Weight=@n4.OutputParam
   )
   n5 = &RigUnit_GetTransform(Item=(Type=Bone,Name="upperarm_l"), Space=LocalSpace, bInitial=False)
   n6 = RigVMVariableNode() {variable=Upperarm_l_pos_ref}
   > &RigUnit_OffsetTransformForItem(
    Item=(Type=Bone,Name="pad_l")
    OffsetTransform=(Rotation=(X=0,Y=-0.061,Z=0,W=0.9981),Translation=(X=-0.8,Y=0,Z=0),Scale3D=(X=1,Y=1.1,Z=1))
    bPropagateToChildren=True
    Weight=&RigVMFunction_MathDoubleRemap(
     SourceMaximum=2
     SourceMinimum=0
     TargetMaximum=1
     TargetMinimum=0
     bClamp=false
     Value=&RigVMFunction_MathVectorDistance(B=@n5.Transform.Translation, A=@n6.Value).Result
    ).Result
   )
   > n7 = &RigUnit_SphericalPoseReader(
    ActiveRegionScaleFactors=(PositiveWidth=1,NegativeWidth=1,PositiveHeight=1,NegativeHeight=1)
    ActiveRegionSize=0.15
    Debug=(bDrawDebug=false,bDraw2D=False,bDrawLocalAxes=False,DebugScale=15,DebugSegments=10,DebugThickness=0.1)
    DriverAxis=(X=1,Y=0,Z=0)
    DriverItem=(Type=Bone,Name="upperarm_l")
    FalloffRegionScaleFactors=(PositiveWidth=1,NegativeWidth=1,PositiveHeight=1,NegativeHeight=1)
    FalloffSize=0.5
    FlipHeightScaling=False
    FlipWidthScaling=False
    OptionalParentItem=(Type=Bone,Name="spine_05")
    RotationOffset=(X=70,Y=0,Z=-20)
   )
   > &RigUnit_OffsetTransformForItem(
    Item=(Type=Bone,Name="pad_l")
    OffsetTransform=(Rotation=(X=0.0015,Y=-0.0349,Z=0.0436,W=0.9984),Translation=(X=-0.6,Y=0.5,Z=0),Scale3D=(X=1,Y=1,Z=1))
    bPropagateToChildren=True
    Weight=@n7.OutputParam
   )
   n8 = &RigUnit_SphericalPoseReader(
    ActiveRegionScaleFactors=(PositiveWidth=1,NegativeWidth=1,PositiveHeight=1,NegativeHeight=1)
    ActiveRegionSize=0.31
    Debug=(bDrawDebug=false,bDraw2D=False,bDrawLocalAxes=False,DebugScale=15,DebugSegments=10,DebugThickness=0.05)
    DriverAxis=(X=-1,Y=0,Z=0)
    DriverItem=(Type=Bone,Name="clavicle_r")
    FalloffRegionScaleFactors=(PositiveWidth=1,NegativeWidth=1,PositiveHeight=1,NegativeHeight=1)
    FalloffSize=0.5
    FlipHeightScaling=False
    FlipWidthScaling=False
    OptionalParentItem=(Type=Bone,Name="None")
    RotationOffset=(X=180,Y=15,Z=0)
    ExecutePin=@RigVMFunction_Sequence.B
   )
   > &RigUnit_OffsetTransformForItem(
    Item=(Type=Bone,Name="MSTR_armorPad_r")
    OffsetTransform=(Rotation=(X=0,Y=-0.1736,Z=0,W=0.9848),Translation=(X=0,Y=0,Z=-2),Scale3D=(X=1,Y=1,Z=1))
    bPropagateToChildren=True
    Weight=@n8.OutputParam
   )
   > &RigUnit_OffsetTransformForItem(
    Item=(Type=Bone,Name="armorPad_front_r")
    OffsetTransform=(Rotation=(X=0,Y=-0.342,Z=0,W=0.9397),Translation=(X=3,Y=0,Z=0),Scale3D=(X=1,Y=1,Z=1))
    bPropagateToChildren=True
    Weight=@n8.OutputParam
   )
   > &RigUnit_OffsetTransformForItem(
    Item=(Type=Bone,Name="armorPad_back_r")
    OffsetTransform=(Rotation=(X=0,Y=-0.2588,Z=0,W=0.9659),Translation=(X=0,Y=0,Z=0),Scale3D=(X=1,Y=1,Z=1))
    bPropagateToChildren=True
    Weight=@n8.OutputParam
   )
   > n9 = &RigUnit_SphericalPoseReader(
    ActiveRegionScaleFactors=(PositiveWidth=1,NegativeWidth=1,PositiveHeight=1,NegativeHeight=1)
    ActiveRegionSize=0.31
    Debug=(bDrawDebug=false,bDraw2D=False,bDrawLocalAxes=False,DebugScale=15,DebugSegments=10,DebugThickness=0.05)
    DriverAxis=(X=1,Y=0,Z=0)
    DriverItem=(Type=Bone,Name="clavicle_l")
    FalloffRegionScaleFactors=(PositiveWidth=1,NegativeWidth=1,PositiveHeight=1,NegativeHeight=1)
    FalloffSize=0.5
    FlipHeightScaling=False
    FlipWidthScaling=False
    OptionalParentItem=(Type=Bone,Name="None")
    RotationOffset=(X=0,Y=15,Z=0)
   )
   > &RigUnit_OffsetTransformForItem(
    Item=(Type=Bone,Name="MSTR_armorPad_l")
    OffsetTransform=(Rotation=(X=0,Y=0,Z=0,W=1),Translation=(X=0,Y=0,Z=0),Scale3D=(X=1,Y=1,Z=1))
    bPropagateToChildren=True
    Weight=@n9.OutputParam
   )
   > &RigUnit_OffsetTransformForItem(
    Item=(Type=Bone,Name="armorPad_front_l")
    OffsetTransform=(Rotation=(X=0,Y=0.1736,Z=0,W=0.9848),Translation=(X=0,Y=0,Z=0),Scale3D=(X=1,Y=1,Z=1))
    bPropagateToChildren=True
    Weight=@n9.OutputParam
   )
   > &RigUnit_OffsetTransformForItem(
    Item=(Type=Bone,Name="armorPad_back_l")
    OffsetTransform=(Rotation=(X=0,Y=0.1736,Z=0,W=0.9848),Translation=(X=0,Y=0,Z=1.5),Scale3D=(X=1,Y=1,Z=1))
    bPropagateToChildren=True
    Weight=@n9.OutputParam
   )
   n10 = &RigUnit_SphericalPoseReader(
    ActiveRegionScaleFactors=(PositiveWidth=1,NegativeWidth=1,PositiveHeight=1,NegativeHeight=1)
    ActiveRegionSize=0.06
    Debug=(bDrawDebug=false,bDraw2D=False,bDrawLocalAxes=False,DebugScale=15,DebugSegments=20,DebugThickness=0.1)
    DriverAxis=(X=1,Y=0,Z=0)
    DriverItem=(Type=Bone,Name="foot_r")
    FalloffRegionScaleFactors=(PositiveWidth=1,NegativeWidth=1,PositiveHeight=1,NegativeHeight=1)
    FalloffSize=0.5
    FlipHeightScaling=False
    FlipWidthScaling=False
    OptionalParentItem=(Type=Bone,Name="None")
    RotationOffset=(X=90,Y=0,Z=0)
    ExecutePin=@RigVMFunction_Sequence.C
   )
   > RigUnit_OffsetTransformForItem = &RigUnit_OffsetTransformForItem(
    Item=(Type=Bone,Name="deform_base_heel_r")
    OffsetTransform=(Rotation=(X=0,Y=0,Z=-0.342,W=0.9397),Translation=(X=0,Y=0,Z=0),Scale3D=(X=1,Y=1,Z=1))
    bPropagateToChildren=True
    Weight=@n10.OutputParam
   )
   > &RigUnit_OffsetTransformForItem(
    Item=(Type=Bone,Name="deform_heel_r")
    OffsetTransform=(Rotation=(X=0,Y=0,Z=0,W=1),Translation=(X=10,Y=0,Z=0),Scale3D=(X=1,Y=1,Z=1))
    bPropagateToChildren=True
    Weight=@n10.OutputParam
   )
   > n11 = &RigUnit_SphericalPoseReader(
    ActiveRegionScaleFactors=(PositiveWidth=1,NegativeWidth=1,PositiveHeight=1,NegativeHeight=1)
    ActiveRegionSize=0.29
    Debug=(bDrawDebug=false,bDraw2D=False,bDrawLocalAxes=False,DebugScale=15,DebugSegments=20,DebugThickness=0.1)
    DriverAxis=(X=1,Y=0,Z=0)
    DriverItem=(Type=Bone,Name="foot_r")
    FalloffRegionScaleFactors=(PositiveWidth=1,NegativeWidth=1,PositiveHeight=1,NegativeHeight=1)
    FalloffSize=0.5
    FlipHeightScaling=False
    FlipWidthScaling=False
    OptionalParentItem=(Type=Bone,Name="None")
    RotationOffset=(X=-90,Y=0,Z=0)
   )
   > &RigUnit_OffsetTransformForItem(
    Item=(Type=Bone,Name="deform_base_heel_r")
    OffsetTransform=(Rotation=(X=0,Y=0,Z=0.1736,W=0.9848),Translation=(X=0,Y=0,Z=0),Scale3D=(X=1,Y=1,Z=1))
    bPropagateToChildren=True
    Weight=@n11.OutputParam
   )
   > &RigUnit_OffsetTransformForItem(
    Item=(Type=Bone,Name="deform_heel_r")
    OffsetTransform=(Rotation=(X=0,Y=0,Z=0,W=1),Translation=(X=-3,Y=0,Z=0),Scale3D=(X=1,Y=1,Z=1))
    bPropagateToChildren=True
    Weight=@n11.OutputParam
   )
   > n12 = &RigUnit_SphericalPoseReader(
    ActiveRegionScaleFactors=(PositiveWidth=1,NegativeWidth=1,PositiveHeight=1,NegativeHeight=1)
    ActiveRegionSize=0.06
    Debug=(bDrawDebug=false,bDraw2D=False,bDrawLocalAxes=False,DebugScale=15,DebugSegments=20,DebugThickness=0.1)
    DriverAxis=(X=-1,Y=0,Z=0)
    DriverItem=(Type=Bone,Name="foot_l")
    FalloffRegionScaleFactors=(PositiveWidth=1,NegativeWidth=1,PositiveHeight=1,NegativeHeight=1)
    FalloffSize=0.5
    FlipHeightScaling=False
    FlipWidthScaling=False
    OptionalParentItem=(Type=Bone,Name="None")
    RotationOffset=(X=-90,Y=0,Z=0)
   )
   > &RigUnit_OffsetTransformForItem(
    Item=(Type=Bone,Name="deform_base_heel_l")
    OffsetTransform=(Rotation=(X=0,Y=0,Z=0.342,W=0.9397),Translation=(X=0,Y=0,Z=0),Scale3D=(X=1,Y=1,Z=1))
    bPropagateToChildren=True
    Weight=@n12.OutputParam
   )
   > &RigUnit_OffsetTransformForItem(
    Item=(Type=Bone,Name="deform_heel_l")
    OffsetTransform=(Rotation=(X=0,Y=0,Z=0,W=1),Translation=(X=10,Y=0,Z=0),Scale3D=(X=1,Y=1,Z=1))
    bPropagateToChildren=True
    Weight=@n12.OutputParam
   )
   > n13 = &RigUnit_SphericalPoseReader(
    ActiveRegionScaleFactors=(PositiveWidth=1,NegativeWidth=1,PositiveHeight=1,NegativeHeight=1)
    ActiveRegionSize=0.29
    Debug=(bDrawDebug=false,bDraw2D=False,bDrawLocalAxes=False,DebugScale=15,DebugSegments=20,DebugThickness=0.1)
    DriverAxis=(X=-1,Y=0,Z=0)
    DriverItem=(Type=Bone,Name="foot_l")
    FalloffRegionScaleFactors=(PositiveWidth=1,NegativeWidth=1,PositiveHeight=1,NegativeHeight=1)
    FalloffSize=0.5
    FlipHeightScaling=False
    FlipWidthScaling=False
    OptionalParentItem=(Type=Bone,Name="None")
    RotationOffset=(X=90,Y=0,Z=0)
   )
   > &RigUnit_OffsetTransformForItem(
    Item=(Type=Bone,Name="deform_base_heel_l")
    OffsetTransform=(Rotation=(X=0,Y=0,Z=-0.1736,W=0.9848),Translation=(X=0,Y=0,Z=0),Scale3D=(X=1,Y=1,Z=1))
    bPropagateToChildren=True
    Weight=@n13.OutputParam
   )
   > &RigUnit_OffsetTransformForItem(
    Item=(Type=Bone,Name="deform_heel_l")
    OffsetTransform=(Rotation=(X=0,Y=0,Z=0,W=1),Translation=(X=-3,Y=0,Z=0),Scale3D=(X=1,Y=1,Z=1))
    bPropagateToChildren=True
    Weight=@n13.OutputParam
   )
   n14 = &RigUnit_SphericalPoseReader(
    ActiveRegionScaleFactors=(PositiveWidth=1,NegativeWidth=1,PositiveHeight=1,NegativeHeight=1)
    ActiveRegionSize=0.05
    Debug=(bDrawDebug=false,bDraw2D=False,bDrawLocalAxes=False,DebugScale=10,DebugSegments=20,DebugThickness=0.05)
    DriverAxis=(X=1,Y=0,Z=0)
    DriverItem=(Type=Bone,Name="lowerarm_r")
    FalloffRegionScaleFactors=(PositiveWidth=1,NegativeWidth=1,PositiveHeight=1,NegativeHeight=1)
    FalloffSize=0.5
    FlipHeightScaling=False
    FlipWidthScaling=False
    OptionalParentItem=(Type=Bone,Name="None")
    RotationOffset=(X=90,Y=0,Z=0)
    ExecutePin=@RigVMFunction_Sequence.D
   )
   > &RigUnit_OffsetTransformForItem(
    Item=(Type=Bone,Name="elbow_pad_r")
    OffsetTransform=(Rotation=(X=0.061,Y=0.0027,Z=0.0435,W=0.9972),Translation=(X=0,Y=1.4,Z=0),Scale3D=(X=1,Y=1,Z=1))
    bPropagateToChildren=True
    Weight=@n14.OutputParam
   )
   > &RigUnit_OffsetTransformForItem(
    Item=(Type=Bone,Name="deform_elbow_in_base_r")
    OffsetTransform=(Rotation=(X=0,Y=0,Z=-0.1736,W=0.9848),Translation=(X=0,Y=-3,Z=0),Scale3D=(X=1,Y=1,Z=1))
    bPropagateToChildren=True
    Weight=@n14.OutputParam
   )
   > n15 = &RigUnit_SphericalPoseReader(
    ActiveRegionScaleFactors=(PositiveWidth=1,NegativeWidth=1,PositiveHeight=1,NegativeHeight=1)
    ActiveRegionSize=0.05
    Debug=(bDrawDebug=false,bDraw2D=False,bDrawLocalAxes=False,DebugScale=10,DebugSegments=20,DebugThickness=0.05)
    DriverAxis=(X=-1,Y=0,Z=0)
    DriverItem=(Type=Bone,Name="lowerarm_l")
    FalloffRegionScaleFactors=(PositiveWidth=1,NegativeWidth=1,PositiveHeight=1,NegativeHeight=1)
    FalloffSize=0.5
    FlipHeightScaling=False
    FlipWidthScaling=False
    OptionalParentItem=(Type=Bone,Name="None")
    RotationOffset=(X=-90,Y=0,Z=0)
   )
   > &RigUnit_OffsetTransformForItem(
    Item=(Type=Bone,Name="elbow_pad_l")
    OffsetTransform=(Rotation=(X=0,Y=0,Z=-0.0436,W=0.999),Translation=(X=0,Y=-1.4,Z=0),Scale3D=(X=1,Y=1,Z=1))
    bPropagateToChildren=True
    Weight=@n15.OutputParam
   )
   > &RigUnit_OffsetTransformForItem(
    Item=(Type=Bone,Name="deform_elbow_in_base_l")
    OffsetTransform=(Rotation=(X=0,Y=0,Z=-0.1736,W=0.9848),Translation=(X=0,Y=3,Z=0),Scale3D=(X=1,Y=1,Z=1))
    bPropagateToChildren=True
    Weight=@n15.OutputParam
   )
   n16 = &RigUnit_SphericalPoseReader(
    ActiveRegionScaleFactors=(PositiveWidth=1,NegativeWidth=1,PositiveHeight=1,NegativeHeight=1)
    ActiveRegionSize=0.1
    Debug=(bDrawDebug=false,bDraw2D=False,bDrawLocalAxes=False,DebugScale=15,DebugSegments=20,DebugThickness=0.1)
    DriverAxis=(X=1,Y=0,Z=0)
    DriverItem=(Type=Bone,Name="calf_r")
    FalloffRegionScaleFactors=(PositiveWidth=1,NegativeWidth=1,PositiveHeight=1,NegativeHeight=1)
    FalloffSize=0.5
    FlipHeightScaling=False
    FlipWidthScaling=False
    OptionalParentItem=(Type=Bone,Name="None")
    RotationOffset=(X=90,Y=0,Z=30)
    ExecutePin=@RigVMFunction_Sequence.E
   )
   > &RigUnit_OffsetTransformForItem(
    Item=(Type=Bone,Name="deform_knee_out_base_r")
    OffsetTransform=(Rotation=(X=0,Y=0,Z=0.5,W=0.866),Translation=(X=0,Y=0,Z=0),Scale3D=(X=1,Y=1,Z=1))
    bPropagateToChildren=True
    Weight=@n16.OutputParam
   )
   > n17 = &RigUnit_SphericalPoseReader(
    ActiveRegionScaleFactors=(PositiveWidth=1,NegativeWidth=1,PositiveHeight=1,NegativeHeight=1)
    ActiveRegionSize=0.1
    Debug=(bDrawDebug=false,bDraw2D=False,bDrawLocalAxes=False,DebugScale=15,DebugSegments=20,DebugThickness=0.1)
    DriverAxis=(X=-1,Y=0,Z=0)
    DriverItem=(Type=Bone,Name="calf_l")
    FalloffRegionScaleFactors=(PositiveWidth=1,NegativeWidth=1,PositiveHeight=1,NegativeHeight=1)
    FalloffSize=0.5
    FlipHeightScaling=False
    FlipWidthScaling=False
    OptionalParentItem=(Type=Bone,Name="None")
    RotationOffset=(X=-90,Y=0,Z=30)
   )
   > &RigUnit_OffsetTransformForItem(
    Item=(Type=Bone,Name="deform_knee_out_base_l")
    OffsetTransform=(Rotation=(X=0,Y=0,Z=0.5,W=0.866),Translation=(X=0,Y=0,Z=0),Scale3D=(X=1,Y=1,Z=1))
    bPropagateToChildren=True
    Weight=@n17.OutputParam
   )
   &RigUnit_AimConstraintLocalSpaceOffset(
    AimAxis=(X=-1,Y=0,Z=0)
    Child=(Type=Bone,Name="Thigh_l_Aim")
    Filter=(bX=True,bY=True,bZ=True)
    Parents=((Item=(Type=Bone,Name="calf_l"),Weight=1))
    UpAxis=(X=0,Y=1,Z=0)
    Weight=1
    WorldUp=(Target=(X=0,Y=0,Z=1),Kind=Location,Space=(Type=Bone,Name="Thigh_l_Aim_TargetSpace"))
    ExecutePin=@RigVMFunction_Sequence.F
   )
   > &RigUnit_AimConstraintLocalSpaceOffset(
    Child=(Type=Bone,Name="Thigh_r_Aim")
    Filter=(bX=True,bY=True,bZ=True)
    Parents=((Item=(Type=Bone,Name="calf_r"),Weight=1))
    UpAxis=(X=0,Y=-1,Z=0)
    Weight=1
    WorldUp=(Target=(X=0,Y=0,Z=1),Kind=Location,Space=(Type=Bone,Name="Thigh_r_Aim_TargetSpace"))
   )
   > n18 = &RigUnit_SphericalPoseReader(
    ActiveRegionScaleFactors=(PositiveWidth=1,NegativeWidth=1,PositiveHeight=1,NegativeHeight=1)
    ActiveRegionSize=0.3
    Debug=(bDrawDebug=True,bDraw2D=False,bDrawLocalAxes=False,DebugScale=10,DebugSegments=20,DebugThickness=0.1)
    DriverAxis=(X=0,Y=0,Z=-1)
    DriverItem=(Type=Bone,Name="thigh_l")
    FalloffRegionScaleFactors=(PositiveWidth=1,NegativeWidth=1,PositiveHeight=1,NegativeHeight=1)
    FalloffSize=0.5
    FlipHeightScaling=False
    FlipWidthScaling=False
    OptionalParentItem=(Type=Bone,Name="Thigh_l_Aim")
    RotationOffset=(X=-90,Y=0,Z=0)
   )
   > n19 = &RigUnit_SphericalPoseReader(
    ActiveRegionScaleFactors=(PositiveWidth=1,NegativeWidth=1,PositiveHeight=1,NegativeHeight=1)
    ActiveRegionSize=0.3
    Debug=(bDrawDebug=True,bDraw2D=False,bDrawLocalAxes=False,DebugScale=10,DebugSegments=20,DebugThickness=0.1)
    DriverAxis=(X=0,Y=0,Z=1)
    DriverItem=(Type=Bone,Name="thigh_r")
    FalloffRegionScaleFactors=(PositiveWidth=1,NegativeWidth=1,PositiveHeight=1,NegativeHeight=1)
    FalloffSize=0.5
    FlipHeightScaling=False
    FlipWidthScaling=False
    OptionalParentItem=(Type=Bone,Name="Thigh_r_Aim")
    RotationOffset=(X=90,Y=0,Z=0)
   )
   Maximum = &RigVMFunction_MathDoubleMax(A=@n19.OutputParam, B=@n18.OutputParam)
   > &RigUnit_OffsetTransformForItem(
    Item=(Type=Bone,Name="deform_glute_m")
    OffsetTransform=(Rotation=(X=0,Y=0,Z=0,W=1),Translation=(X=9,Y=0,Z=0),Scale3D=(X=1,Y=1,Z=1))
    bPropagateToChildren=True
    Weight=@Maximum.Result
   )
   &RigUnit_OffsetTransformForItem(
    ExecutePin=()
    Item=(Type=Bone,Name="clavicle_r")
    OffsetTransform=(Rotation=(X=0,Y=0,Z=0.1736,W=0.9848),Translation=(X=0,Y=0,Z=0),Scale3D=(X=1,Y=1,Z=1))
    Weight=1
    bPropagateToChildren=True
   )
   > &RigUnit_OffsetTransformForItem(
    Item=(Type=Bone,Name="upperarm_r")
    OffsetTransform=(Rotation=(X=0.0616,Y=-0.0616,Z=0.7044,W=0.7044),Translation=(X=0,Y=0,Z=0),Scale3D=(X=1,Y=1,Z=1))
    Weight=1
    bPropagateToChildren=True
   )
   &RigUnit_SetTransform(
    ExecutePin=()
    Item=(Type=Bone,Name="foot_r")
    Weight=1
    bInitial=False
    Value=&RigUnit_GetTransform(Item=(Type=Control,Name="foot_r_ctrl"), bInitial=False).Transform
   )
   > &RigUnit_SetTransform(
    Item=(Type=Bone,Name="calf_r")
    Weight=1
    bInitial=False
    Value=&RigUnit_GetTransform(Item=(Type=Control,Name="calf_r_ctrl"), bInitial=False).Transform
   )
   > &RigUnit_SetTransform(
    Item=(Type=Bone,Name="lowerarm_r")
    Weight=1
    bInitial=False
    Value=&RigUnit_GetTransform(Item=(Type=Control,Name="lowerarm_r_ctrl"), bInitial=False).Transform
   )
   > &RigUnit_SetTransform(
    Item=(Type=Bone,Name="clavicle_r")
    Weight=1
    bInitial=False
    Value=&RigUnit_GetTransform(Item=(Type=Control,Name="clavicle_r_ctrl"), bInitial=False).Transform
   )
   RigUnit_GetTransform = &RigUnit_GetTransform(Item=(Type=Control,Name="upperarm_r_ctrl"), bInitial=False)
   > RigUnit_SetTransform = &RigUnit_SetTransform(Item=(Type=Bone,Name="upperarm_r"), Weight=1, bInitial=False, Value=@RigUnit_GetTransform.Transform)
   > &RigUnit_OffsetTransformForItem(
    Item=(Type=Bone,Name="calf_l")
    OffsetTransform=(Rotation=(X=0,Y=0,Z=0.866,W=0.5),Translation=(X=0,Y=0,Z=0),Scale3D=(X=1,Y=1,Z=1))
    Weight=1
    bPropagateToChildren=True
   )
   > &RigUnit_OffsetTransformForItem(
    Item=(Type=Bone,Name="clavicle_l")
    OffsetTransform=(Rotation=(X=0,Y=0,Z=0.3827,W=0.9239),Translation=(X=0,Y=0,Z=0),Scale3D=(X=1,Y=1,Z=1))
    Weight=1
    bPropagateToChildren=True
   )
   > &RigUnit_OffsetTransformForItem(
    Item=(Type=Bone,Name="upperarm_l")
    OffsetTransform=(Rotation=(X=0,Y=0,Z=0.3827,W=0.9239),Translation=(X=0,Y=0,Z=0),Scale3D=(X=1,Y=1,Z=1))
    Weight=1
    bPropagateToChildren=True
   )
   > &RigUnit_OffsetTransformForItem(
    Item=(Type=Bone,Name="lowerarm_l")
    OffsetTransform=(Rotation=(X=0,Y=0,Z=0.5,W=0.866),Translation=(X=0,Y=0,Z=0),Scale3D=(X=1,Y=1,Z=1))
    Weight=1
    bPropagateToChildren=True
   )
  ]}
]
hierarchy [
 root {loc=<0 0 0>, rot=<0 0 0 1>}
 pelvis {parent=root, loc=<0 -0.5301 98.0571>, rot=<0.0127 0.707 -0.0127 -0.707>}
 spine_01 {parent=pelvis, loc=<6.1208 -0.9328 0>, rot=<0 0 0.034 -0.9994>}
 spine_02 {parent=spine_01, loc=<6.5997 -0.2095 0>, rot=<0 0 0.0238 -0.9997>}
 spine_03 {parent=spine_02, loc=<6.4914 -0.1237 0>, rot=<0 0 0.0149 -0.9999>}
 spine_04 {parent=spine_03, loc=<6.9598 0.0794 0>, rot=<0 0 -0.0132 -0.9999>}
 spine_05 {parent=spine_04, loc=<6.9987 0 0>, rot=<0 0 -0.0015 -1>}
 attach_cape {parent=spine_05, loc=<9.3216 -8.4501 0>, rot=<0.0616 -0.7044 -0.0616 -0.7044>}
 neck_01 {parent=spine_05, loc=<11.3138 0.1529 0>, rot=<0 0 -0.2003 -0.9797>}
 neck_02 {parent=neck_01, loc=<4.5627 0 0>, rot=<0 0 0.0504 -0.9987>}
 head {parent=neck_02, loc=<5.3914 0 0>, rot=<0 0 0.115 -0.9934>}
 MSTR_Hair_r_2 {parent=head, loc=<13.9582 -6.6499 -4.5374>, rot=<0.3835 -0.5267 0.3589 -0.6684>}
 Hair_r_2 {parent=MSTR_Hair_r_2, loc=<0 0 0>, rot=<0 0 0 1>}
 Hair_r_3 {parent=Hair_r_2, loc=<4.8797 0 0.0002>, rot=<-0.1062 -0.4961 0.1236 -0.8528>}
 Hair_r_4 {parent=Hair_r_3, loc=<6.1622 -0.0002 -0.0005>, rot=<-0.0304 -0.277 0.1358 -0.9507>}
 Hair_r_5 {parent=Hair_r_4, loc=<7.7257 0 0>, rot=<0.0061 -0.1232 0.0391 -0.9916>}
 Hair_r_6 {parent=Hair_r_5, loc=<8.2864 0 0>, rot=<0.0155 0.0731 -0.054 -0.9957>}
 Hair_r_7 {parent=Hair_r_6, loc=<9.6565 0 0>, rot=<-0.003 0.0537 -0.023 -0.9983>}
 Hair_r_8 {parent=Hair_r_7, loc=<7.3763 0 0>, rot=<0.0014 0.0579 0.0161 -0.9982>}
 Hair_r_9 {parent=Hair_r_8, loc=<8.4829 0 0>, rot=<-0.0007 0.0122 0.0539 -0.9985>}
 MSTR_Hair_l_2 {parent=head, loc=<13.958 -6.65 4.5374>, rot=<0.6684 -0.3589 0.5267 -0.3835>}
 Hair_l_2 {parent=MSTR_Hair_l_2, loc=<0 0 0>, rot=<0 0 0 1>}
 Hair_l_3 {parent=Hair_l_2, loc=<4.8796 0 0>, rot=<0.1093 -0.4955 -0.1182 -0.8536>}
 Hair_l_4 {parent=Hair_l_3, loc=<6.1616 0 0>, rot=<0.0303 -0.2764 -0.1464 -0.9493>}
 Hair_l_5 {parent=Hair_l_4, loc=<7.7457 0 0>, rot=<-0.0063 -0.1239 -0.0342 -0.9917>}
 Hair_l_6 {parent=Hair_l_5, loc=<8.2862 0 0>, rot=<-0.0155 0.0731 0.054 -0.9957>}
 Hair_l_7 {parent=Hair_l_6, loc=<9.6566 0 0>, rot=<0.003 0.0537 0.023 -0.9983>}
 Hair_l_8 {parent=Hair_l_7, loc=<7.3768 0 0>, rot=<-0.0014 0.0579 -0.0161 -0.9982>}
 Hair_l_9 {parent=Hair_l_8, loc=<8.4822 0 0>, rot=<0.0007 0.0122 -0.0539 -0.9985>}
 attach_backpack {parent=spine_05, loc=<0 -8.35 0>, rot=<-0.0093 -0.707 0.0093 -0.707>}
 clavicle_l {parent=spine_05, loc=<5.4831 -0.1053 -3.2946>, rot=<0.0465 -0.7289 -0.009 -0.6829>}
 upperarm_l {parent=clavicle_l, loc=<9.8658 -1.2136 -0.4105>, rot=<0.0164 -0.3245 -0.0269 -0.9454>}
 lowerarm_l {parent=upperarm_l, loc=<26.465 -0.8345 0.259>, rot=<0.0036 0.0161 -0.2298 -0.9731>}
 hand_l {parent=lowerarm_l, loc=<22.7355 -0.3203 -0.6105>, rot=<-0.6784 -0.0223 0.0735 -0.7307>}
 hand_attach_l {parent=hand_l, loc=<8.5 -3 0>, rot=<0.7071 0.7071 0 0>}
 weapon_l {parent=hand_l, loc=<-0.1974 -3.4111 0.3811>, rot=<0 0 0 1>}
 thumb_01_l {parent=hand_l, loc=<1.3118 -0.3191 -2.459>, rot=<0.6546 -0.4042 0.0664 -0.6354>}
 thumb_02_l {parent=thumb_01_l, loc=<5.9336 -0.0448 0.0401>, rot=<-0.0037 0.0213 0.298 -0.9543>}
 thumb_03_l {parent=thumb_02_l, loc=<2.2223 0.3682 -0.0907>, rot=<0.0118 -0.013 0.0066 -0.9998>}
 index_metacarpal_l {parent=hand_l, loc=<1.6535 0.0183 -2.96>, rot=<0.0158 -0.0124 -0.0173 -0.9996>}
 index_01_l {parent=index_metacarpal_l, loc=<8.3008 0.0357 -0.0524>, rot=<-0.0044 0.0485 0.0924 -0.9945>}
 index_02_l {parent=index_01_l, loc=<3.3587 -0.3004 0.0452>, rot=<-0.002 0.0027 0.3225 -0.9466>}
 index_03_l {parent=index_02_l, loc=<2.5578 -0.123 -0.0252>, rot=<-0.002 0.0067 0.2558 -0.9667>}
 middle_metacarpal_l {parent=hand_l, loc=<2.1856 -0.1725 -0.5984>, rot=<0.006 0.0168 -0.0329 -0.9993>}
 middle_01_l {parent=middle_metacarpal_l, loc=<7.3275 -0.2447 0.0129>, rot=<-0.0208 0.0042 0.0858 -0.9961>}
 middle_02_l {parent=middle_01_l, loc=<4.0123 0.1401 0.1175>, rot=<-0.0218 0.0258 0.3512 -0.9357>}
 middle_03_l {parent=middle_02_l, loc=<2.6077 0.0534 0.1373>, rot=<0.0082 0.0188 0.1708 -0.9851>}
 ring_metacarpal_l {parent=hand_l, loc=<2.2074 -0.3625 1.3638>, rot=<-0.0284 0.0722 -0.0458 -0.9959>}
 ring_01_l {parent=ring_metacarpal_l, loc=<7.0295 -0.6197 -0.1446>, rot=<-0.0489 -0.0252 0.1892 -0.9804>}
 ring_02_l {parent=ring_01_l, loc=<3.6174 0.4647 -0.1027>, rot=<0 0 0.2806 -0.9598>}
 ring_03_l {parent=ring_02_l, loc=<2.4152 0.114 0.0199>, rot=<0 0 0.1189 -0.9929>}
 pinky_metacarpal_l {parent=hand_l, loc=<1.7616 -0.9753 3.3814>, rot=<-0.0397 0.0814 -0.069 -0.9935>}
 pinky_01_l {parent=pinky_metacarpal_l, loc=<7.1709 -0.4831 -0.1094>, rot=<-0.0654 -0.0035 0.2242 -0.9723>}
 pinky_02_l {parent=pinky_01_l, loc=<2.8274 0.0893 -0.1043>, rot=<0.0037 0.0011 0.2854 -0.9584>}
 pinky_03_l {parent=pinky_02_l, loc=<2.2046 0.0307 -0.0078>, rot=<-0.0002 0.0098 0.1445 -0.9895>}
 deform_wrist_dn_base_l {parent=hand_l, loc=<0 -1.5326 0>, rot=<0 0 0 -1>}
 deform_wrist_dn_l {parent=deform_wrist_dn_base_l, loc=<0 0 0>, rot=<0 0 0 1>}
 deform_wrist_up_base_l {parent=hand_l, loc=<0 2.3009 0>, rot=<0 0 0 -1>}
 deform_wrist_up_l {parent=deform_wrist_up_base_l, loc=<0 0 0>, rot=<0 0 0 1>}
 lowerarm_twist_02_l {parent=lowerarm_l, loc=<10.6707 0 0>, rot=<0 0 0 1>}
 lowerarm_twist_01_l {parent=lowerarm_l, loc=<18.316 0 0>, rot=<0 0 0 1>}
 deform_elbow_in_base_l {parent=lowerarm_l, loc=<0.1834 3.6062 -0.2572>, rot=<0 0 0 -1>}
 deform_elbow_in_l {parent=deform_elbow_in_base_l, loc=<0 0 0>, rot=<0 0 0 1>}
 deform_elbow_out_base_l {parent=lowerarm_l, loc=<0 -4 0>, rot=<0 0 0 1>}
 deform_elbow_out_l {parent=deform_elbow_out_base_l, loc=<0 0 0>, rot=<0 0 0 1>}
 elbow_pad_l {parent=lowerarm_l, loc=<-1.14 -1.409 3.0436>, rot=<0 -0.9371 0.3491 0>}
 elbow_l_fx {parent=lowerarm_l, loc=<0 -3 4>, rot=<0 0 0 1>}
 upperarm_twist_01_l {parent=upperarm_l, loc=<7.2135 -0.1134 0>, rot=<0 0 0 1>}
 deform_pad_l {parent=upperarm_twist_01_l, loc=<-4.1511 -0.3249 3.3924>, rot=<0.7691 0.0158 -0.6387 -0.0131>}
 pad_l {parent=deform_pad_l, loc=<-0.3609 0.2929 -0.0036>, rot=<-0.0009 0.0208 -0.026 -0.9994>}
 deform_upperarm_bk_base_l {parent=upperarm_twist_01_l, loc=<-6.4053 -4.3746 0.314>, rot=<0 0 0 -1>}
 deform_upperarm_bk_l {parent=deform_upperarm_bk_base_l, loc=<0 0 0>, rot=<0 0 0 1>}
 deform_upperarm_fr_base_l {parent=upperarm_twist_01_l, loc=<-5.6312 5.092 0.3354>, rot=<0 0 0 -1>}
 deform_upperarm_fr_l {parent=deform_upperarm_fr_base_l, loc=<0 0 0>, rot=<0 0 0 1>}
 deform_upperarm_up_base_l {parent=upperarm_twist_01_l, loc=<-8.891 0.4825 3.9517>, rot=<0 0 0 -1>}
 deform_upperarm_up_l {parent=deform_upperarm_up_base_l, loc=<0 0 0>, rot=<0 0 0 1>}
 upperarm_twist_02_l {parent=upperarm_l, loc=<19.7998 -0.1134 0>, rot=<0 0 0 1>}
 MSTR_armorPad_l {parent=clavicle_l, loc=<4.7051 -1.9365 5.6062>, rot=<0.7951 0.0028 0.6065 0.0026>}
 armorPad_front_l {parent=MSTR_armorPad_l, loc=<0.2961 0 -1.6145>, rot=<0 0 0 1>}
 armorPad_back_l {parent=MSTR_armorPad_l, loc=<0.2955 0 1.6076>, rot=<0 0 0 1>}
 clavicle_r {parent=spine_05, loc=<5.4831 -0.1053 3.2946>, rot=<0.7289 0.0465 -0.6829 0.009>}
 upperarm_r {parent=clavicle_r, loc=<-9.8658 1.2136 0.4105>, rot=<0.0164 -0.3245 -0.0269 -0.9454>}
 lowerarm_r {parent=upperarm_r, loc=<-26.465 0.8345 -0.259>, rot=<0.0036 0.0161 -0.2298 -0.9731>}
 hand_r {parent=lowerarm_r, loc=<-22.7355 0.3203 0.6106>, rot=<-0.6784 -0.0223 0.0735 -0.7307>}
 thumb_01_r {parent=hand_r, loc=<-1.3119 0.3192 2.459>, rot=<0.6546 -0.4042 0.0664 -0.6354>}
 thumb_02_r {parent=thumb_01_r, loc=<-5.9336 0.0448 -0.0401>, rot=<-0.0037 0.0213 0.298 -0.9543>}
 thumb_03_r {parent=thumb_02_r, loc=<-2.2223 -0.3682 0.0907>, rot=<0.0118 -0.013 0.0066 -0.9998>}
 index_metacarpal_r {parent=hand_r, loc=<-1.6534 -0.0181 2.9601>, rot=<0.0158 -0.0124 -0.0173 -0.9996>}
 index_01_r {parent=index_metacarpal_r, loc=<-8.3008 -0.0357 0.0524>, rot=<-0.0044 0.0485 0.0924 -0.9945>}
 index_02_r {parent=index_01_r, loc=<-3.3589 0.3003 -0.0453>, rot=<-0.002 0.0027 0.3225 -0.9466>}
 index_03_r {parent=index_02_r, loc=<-2.5578 0.123 0.0252>, rot=<-0.002 0.0067 0.2558 -0.9667>}
 middle_metacarpal_r {parent=hand_r, loc=<-2.1856 0.1726 0.5984>, rot=<0.006 0.0168 -0.0329 -0.9993>}
 middle_01_r {parent=middle_metacarpal_r, loc=<-7.3273 0.2447 -0.0129>, rot=<-0.0208 0.0042 0.0858 -0.9961>}
 middle_02_r {parent=middle_01_r, loc=<-4.0124 -0.14 -0.1175>, rot=<-0.0218 0.0258 0.3512 -0.9357>}
 middle_03_r {parent=middle_02_r, loc=<-2.6077 -0.0534 -0.1373>, rot=<0.0082 0.0188 0.1708 -0.9851>}
 ring_metacarpal_r {parent=hand_r, loc=<-2.2073 0.3627 -1.3638>, rot=<-0.0284 0.0722 -0.0458 -0.9959>}
 ring_01_r {parent=ring_metacarpal_r, loc=<-7.0297 0.6197 0.1446>, rot=<-0.0489 -0.0252 0.1892 -0.9804>}
 ring_02_r {parent=ring_01_r, loc=<-3.6171 -0.4648 0.1027>, rot=<0 0 0.2806 -0.9598>}
 ring_03_r {parent=ring_02_r, loc=<-2.4152 -0.114 -0.0199>, rot=<0 0 0.1189 -0.9929>}
 pinky_metacarpal_r {parent=hand_r, loc=<-1.7617 0.9753 -3.3814>, rot=<-0.0397 0.0814 -0.069 -0.9935>}
 pinky_01_r {parent=pinky_metacarpal_r, loc=<-7.1707 0.4832 0.1094>, rot=<-0.0654 -0.0035 0.2242 -0.9723>}
 pinky_02_r {parent=pinky_01_r, loc=<-2.8274 -0.0893 0.1043>, rot=<0.0037 0.0011 0.2854 -0.9584>}
 pinky_03_r {parent=pinky_02_r, loc=<-2.2045 -0.031 0.0078>, rot=<-0.0002 0.0098 0.1445 -0.9895>}
 weapon_r {parent=hand_r, loc=<0.1974 3.4111 -0.3811>, rot=<0 0 0 1>}
 hand_attach_r {parent=hand_r, loc=<-8.5001 2.9999 0>, rot=<0 0 -0.7071 -0.7071>}
 deform_wrist_dn_base_r {parent=hand_r, loc=<-0.0001 1.5328 0>, rot=<0 0 0 -1>}
 deform_wrist_dn_r {parent=deform_wrist_dn_base_r, loc=<0 0 0>, rot=<0 0 0 1>}
 deform_wrist_up_base_r {parent=hand_r, loc=<-0.0002 -2.3007 -0.0001>, rot=<0 0 0 -1>}
 deform_wrist_up_r {parent=deform_wrist_up_base_r, loc=<0 0 0>, rot=<0 0 0 1>}
 lowerarm_twist_02_r {parent=lowerarm_r, loc=<-10.6707 0 0>, rot=<0 0 0 -1>}
 lowerarm_twist_01_r {parent=lowerarm_r, loc=<-18.316 0 0>, rot=<0 0 0 -1>}
 deform_elbow_in_base_r {parent=lowerarm_r, loc=<-0.1834 -3.6062 0.2572>, rot=<0 0 0 -1>}
 deform_elbow_in_r {parent=deform_elbow_in_base_r, loc=<0 0 0>, rot=<0 0 0 1>}
 deform_elbow_out_base_r {parent=lowerarm_r, loc=<0 4 -0.0001>, rot=<0 0 0 -1>}
 deform_elbow_out_r {parent=deform_elbow_out_base_r, loc=<0 0 0>, rot=<0 0 0 1>}
 elbow_pad_r {parent=lowerarm_r, loc=<1.1401 1.409 -3.0436>, rot=<0.3491 0 0 -0.9371>}
 elbow_r_fx {parent=lowerarm_r, loc=<0 3 -4>, rot=<0 0 0 1>}
 upperarm_twist_01_r {parent=upperarm_r, loc=<-7.2135 0.1134 0>, rot=<0 0 0 -1>}
 deform_pad_r {parent=upperarm_twist_01_r, loc=<4.1514 0.3249 -3.3927>, rot=<-0.0158 0.7691 0.0131 -0.6387>}
 pad_r {parent=deform_pad_r, loc=<0 0 0>, rot=<0 0 0 1>}
 deform_upperarm_bk_base_r {parent=upperarm_twist_01_r, loc=<6.4053 4.3746 -0.314>, rot=<0 0 0 -1>}
 deform_upperarm_bk_r {parent=deform_upperarm_bk_base_r, loc=<0 0 0>, rot=<0 0 0 1>}
 deform_upperarm_fr_base_r {parent=upperarm_twist_01_r, loc=<5.6312 -5.092 -0.3354>, rot=<0 0 0 -1>}
 deform_upperarm_fr_r {parent=deform_upperarm_fr_base_r, loc=<0 0 0>, rot=<0 0 0 1>}
 deform_upperarm_up_base_r {parent=upperarm_twist_01_r, loc=<8.891 -0.4825 -3.9517>, rot=<0 0 0 -1>}
 deform_upperarm_up_r {parent=deform_upperarm_up_base_r, loc=<0 0 0>, rot=<0 0 0 1>}
 upperarm_twist_02_r {parent=upperarm_r, loc=<-19.7998 0.1134 0>, rot=<0 0 0 -1>}
 MSTR_armorPad_r {parent=clavicle_r, loc=<-4.0179 1.936 -5.7941>, rot=<0.0028 -0.7951 0.0026 -0.6065>}
 armorPad_front_r {parent=MSTR_armorPad_r, loc=<0.2963 0 1.6146>, rot=<0 0 0 1>}
 armorPad_back_r {parent=MSTR_armorPad_r, loc=<0.2963 0 -1.6073>, rot=<0 0 0 1>}
 hair_alt_1 {parent=spine_05, loc=<8.2925 -11.2555 0>, rot=<0 0 -0.081 -0.9967>}
 deform_pec_aim_base_l {parent=spine_04, loc=<3.2456 9.3309 -1.2237>, rot=<0.0603 -0.4033 0.3318 -0.8507>}
 deform_pec_aim_l {parent=deform_pec_aim_base_l, loc=<0 0 0>, rot=<0 0 0 1>}
 deform_pec_base_l {parent=deform_pec_aim_l, loc=<9.3665 0 0>, rot=<0 0 0 1>}
 deform_pec_l {parent=deform_pec_base_l, loc=<0 0 0>, rot=<0 0 0 1>}
 deform_pec_aim_base_r {parent=spine_04, loc=<3.2452 9.3309 1.2237>, rot=<-0.4033 -0.0603 0.8507 0.3318>}
 deform_pec_aim_r {parent=deform_pec_aim_base_r, loc=<0 0 0>, rot=<0 0 0 1>}
 deform_pec_base_r {parent=deform_pec_aim_r, loc=<-9.3667 -0.0003 -0.0004>, rot=<0 0 0 1>}
 deform_pec_r {parent=deform_pec_base_r, loc=<0 0 0>, rot=<0 0 0 1>}
 hair_alt_2 {parent=spine_03, loc=<6.3216 -14.1858 0>, rot=<0 0 -0.0956 -0.9954>}
 hair_alt_3 {parent=spine_02, loc=<0.4112 -18.1307 0>, rot=<0 0 -0.148 -0.989>}
 thigh_l {parent=pelvis, loc=<-1.0855 -0.2109 -8.7053>, rot=<0.0009 0.0355 0.0155 -0.9993>}
 calf_l {parent=thigh_l, loc=<-42.5284 -0.0019 -0.7297>, rot=<0.0006 -0.0161 -0.0391 -0.9991>}
 foot_l {parent=calf_l, loc=<-46.7401 -2.5014 -0.592>, rot=<0.0427 -0.0238 0.0353 -0.9982>}
 ball_l {parent=foot_l, loc=<-3.5658 10.8066 -0.1544>, rot=<0.0001 -0.0001 -0.7063 -0.7079>}
 deform_base_heel_l {parent=foot_l, loc=<2.2762 -0.6819 -0.0001>, rot=<0.7039 -0.7103 0 0>}
 deform_heel_l {parent=deform_base_heel_l, loc=<0 0 0>, rot=<0 0 0 1>}
 joint3 {parent=foot_l, loc=<1.5 7 0>, rot=<0 0 0 1>}
 calf_twist_02_l {parent=calf_l, loc=<-19.1546 -1.7943 -0.2638>, rot=<0.0237 0.0095 -0.0035 -0.9997>}
 calf_twist_01_l {parent=calf_l, loc=<-35.8483 -1.9197 -0.5794>, rot=<0.0237 0.0095 -0.0035 -0.9997>}
 deform_knee_in_base_l {parent=calf_l, loc=<0.8031 -5.0231 0.2456>, rot=<0 0 0 -1>}
 deform_knee_in_l {parent=deform_knee_in_base_l, loc=<0 0 0>, rot=<0 0 0 1>}
 deform_knee_out_base_l {parent=calf_l, loc=<0 4 0>, rot=<0 0 0 -1>}
 deform_knee_out_l {parent=deform_knee_out_base_l, loc=<0 0 0>, rot=<0 0 0 1>}
 knee_l_fx {parent=calf_l, loc=<-4 1.5 0>, rot=<0 0 0 1>}
 thigh_twist_01_l {parent=thigh_l, loc=<-27.952 0 0>, rot=<0.0013 0 0 -1>}
 deform_groin_base_l {parent=thigh_l, loc=<0.4801 5.6776 -0.2486>, rot=<0 0 0 -1>}
 deform_groin_l {parent=deform_groin_base_l, loc=<0 0 0>, rot=<0 0 0 1>}
 thigh_r {parent=pelvis, loc=<-1.0853 -0.2109 8.7053>, rot=<0.0355 -0.0009 0.9993 0.0155>}
 calf_r {parent=thigh_r, loc=<42.5286 0.0019 0.7297>, rot=<0.0006 -0.0161 -0.0391 -0.9991>}
 foot_r {parent=calf_r, loc=<46.7403 2.5014 0.592>, rot=<0.0427 -0.0238 0.0353 -0.9982>}
 ball_r {parent=foot_r, loc=<3.5658 -10.8066 0.1544>, rot=<0.0001 -0.0001 -0.7063 -0.7079>}
 deform_base_heel_r {parent=foot_r, loc=<-2.2764 0.6819 0>, rot=<0 0 -0.7039 -0.7103>}
 deform_heel_r {parent=deform_base_heel_r, loc=<0 0 0>, rot=<0 0 0 1>}
 foot_r_fx {parent=foot_r, loc=<-1.5 -7 0>, rot=<0 0 0 1>}
 calf_twist_02_r {parent=calf_r, loc=<19.1546 1.7943 0.2638>, rot=<0.0237 0.0095 -0.0035 -0.9997>}
 calf_twist_01_r {parent=calf_r, loc=<35.8485 1.9197 0.5794>, rot=<0.0237 0.0095 -0.0035 -0.9997>}
 deform_knee_in_base_r {parent=calf_r, loc=<-0.8032 5.0231 -0.2457>, rot=<0 0 0 -1>}
 deform_knee_in_r {parent=deform_knee_in_base_r, loc=<0 0 0>, rot=<0 0 0 1>}
 deform_knee_out_base_r {parent=calf_r, loc=<-0.0001 -4 0>, rot=<0 0 0 -1>}
 deform_knee_out_r {parent=deform_knee_out_base_r, loc=<0 0 0>, rot=<0 0 0 1>}
 knee_r_fx {parent=calf_r, loc=<4 -1.5 0>, rot=<0 0 0 1>}
 thigh_twist_01_r {parent=thigh_r, loc=<27.952 0 0>, rot=<0.0013 0 0 -1>}
 deform_groin_base_r {parent=thigh_r, loc=<-0.4801 -5.6776 0.2486>, rot=<0 0 0 -1>}
 deform_groin_r {parent=deform_groin_base_r, loc=<0 0 0>, rot=<0 0 0 1>}
 deform_glute_base_l {parent=pelvis, loc=<-7.3466 -9.0213 -9.1245>, rot=<0.0172 0.031 -0.4477 -0.8935>}
 deform_glute_l {parent=deform_glute_base_l, loc=<0 0 0>, rot=<0 0 0 1>}
 deform_glute_base_r {parent=pelvis, loc=<-7.3464 -9.0213 9.1245>, rot=<0.031 -0.0172 0.8935 -0.4477>}
 deform_glute_r {parent=deform_glute_base_r, loc=<0 0 0>, rot=<0 0 0 1>}
 deform_glute_base_m {parent=pelvis, loc=<-7.148 -9.3728 0>, rot=<0.179 -0.9838 0 0>}
 deform_glute_m {parent=deform_glute_base_m, loc=<0 0 0>, rot=<0 0 0 1>}
 Thigh_l_Aim {parent=pelvis, loc=<-1.0855 -0.2109 -8.7053>, rot=<0.0009 0.0355 0.0155 -0.9993>}
 Thigh_r_Aim {parent=pelvis, loc=<-1.0853 -0.2109 8.7053>, rot=<0.0355 -0.0009 0.9993 0.0155>}
 Thigh_l_Aim_TargetSpace {parent=pelvis, loc=<0.5 1 -8>, rot=<-0.0009 -0.0355 -0.0155 0.9993>}
 Thigh_r_Aim_TargetSpace {parent=pelvis, loc=<0.5 1 8>, rot=<0.0355 -0.0009 0.9993 0.0155>}
 ik_hand_root {parent=root, loc=<0 0 0>, rot=<0 0 0 1>}
 ik_hand_gun {parent=ik_hand_root, loc=<-48.1744 3.3546 104.4381>, rot=<0.6276 0.1309 0.4183 -0.6434>}
 ik_hand_r {parent=ik_hand_gun, loc=<0 0 0>, rot=<0 0 0 1>}
 ik_hand_l {parent=ik_hand_gun, loc=<59.3271 67.6997 34.3524>, rot=<-0.6158 -0.7027 -0.3565 0>}
 ik_foot_root {parent=root, loc=<0 0 0>, rot=<0 0 0 1>}
 ik_foot_l {parent=ik_foot_root, loc=<14.8262 -7.3545 8.296>, rot=<0.0357 0.7045 0.0284 -0.7082>}
 ik_foot_r {parent=ik_foot_root, loc=<-14.8262 -7.3546 8.2958>, rot=<0.7082 0.0284 -0.7045 0.0357>}
 root_fp {parent=root, loc=<0 0 0>, rot=<0 0 0 1>}
 attach_fp {parent=root_fp, loc=<0 0 0>, rot=<0 0 0 1>}
 camera_root {parent=root_fp, loc=<0 0 0>, rot=<0 0 0 1>}
 camera {parent=camera_root, loc=<0 0 0>, rot=<0 0 0 1>}
 clavicle_r_ctrl {controlType=EULER_TRANSFORM, type=Control, loc=<0 0 0>, rot=<0 0 0 1>}
 upperarm_r_ctrl {controlType=EULER_TRANSFORM, parent=clavicle_r_ctrl, type=Control, loc=<0 0 0>, rot=<0 0 0 1>}
 lowerarm_r_ctrl {controlType=EULER_TRANSFORM, parent=upperarm_r_ctrl, type=Control, loc=<0 0 0>, rot=<0 0 0 1>}
 calf_r_ctrl {controlType=EULER_TRANSFORM, type=Control, loc=<0 0 0>, rot=<0 0 0 1>}
 foot_r_ctrl {controlType=EULER_TRANSFORM, parent=calf_r_ctrl, type=Control, loc=<0 0 0>, rot=<0 0 0 1>}
]
variables [
 Upperarm_l_pos_ref {type=FVector, default=(X=0,Y=0,Z=0)}
 Upperarm_r_pos_ref {type=FVector, default=(X=0,Y=0,Z=0)}
]
