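# Animation Blueprint for the F_MED_TruckScale female player body that adjusts backpack placement for a specific hair face accessory.
# It copies the pose from the attached parent mesh, additively offsets attach_backpack when MoveBackpackBack is true, then runs the TruckScale Control Rig.
# On customization load, that flag is true if the Face mesh is the TruckScale hair accessory and the backpack lacks the TruckScale meta tag.
# MoveBackpackForward stores the inverse but is unused in the AnimGraph.
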AnimBlueprint /BRCosmetics/Characters/Player/Female/Medium/Bodies/F_MED_TruckScale/Meshes/F_MED_TruckScale_AnimBP
target_skeleton=&/F_MED_TruckScale_Skeleton
graphs [
 AnimGraph {nodes [
   n1 = "Transform (Modify) Bone - Bone: attach_backpack"!(
    ComponentPose=Local To Component(LocalPose=Copy Pose From Mesh() {bUseAttachedParent=true})
    bAlphaBoolEnabled:bool=Get MoveBackpackBack() -> (MoveBackpackBack:bool)
    Translation="0,-3.5,0"
    Rotation="0,0,0"
    Scale="1,1,1"
   ) -> (Pose) {
    alphaInputType=Bool
    translationMode=BMM_Additive
    alphaBoolBlend {blendOption=Linear}
    boneToModify {boneName=attach_backpack}
    translation {y=-3.5}
   }
   n2 = Control Rig(Source=Component To Local(ComponentPose=@n1)) {
    alphaBoolBlend {blendOption=Linear}
    controlRigAssetReference {blueprintRigClass=&/F_MED_TruckScale_CtrlRig.F_MED_TruckScale_CtrlRig_C}
   }
   Output Pose(Result=@n2) {layerGroup=None}
  ]}
 EventGraph {nodes [
   Event OnCharacterCustomizationLoaded()
   > n1 = Cast To PlayerPawn_Athena(Object=TryGetPawnOwner())
   n2 = BackpackItemDefinitionHasAnyMetaTag!(Tags:Array of Gameplay Tag Structures=Make Array("[0]"=(TagName="Cosmetics.MetaTags.BackPack_TruckScale")), bRequiresEquipped:bool=false) -> (ReturnValue:bool)
   n3 = Get SkeletalMesh!(self:Skinned Mesh Component Object Reference=GetSkeletalMeshForPartType(self=@n1, PartType=Face)) -> (SkeletalMesh:Skeletal Mesh Object Reference)
   > n4 = Set MoveBackpackBack(MoveBackpackBack:bool=AND Boolean(
     A="Equal (Object)"(A=@n3, B=/BRCosmetics/Characters/Player/Female/Medium/Bodies/F_MED_TruckScale_Hair/Meshes/Parts/F_MED_TruckScale_Hair_FaceAcc)
     B=NOT Boolean(A=@n2)
    )) -> (Output_Get:bool)
   > Set MoveBackpackForward(MoveBackpackForward:bool=NOT Boolean(A=@n4)) -> (Output_Get:bool)
   Event BlueprintUpdateAnimation()
   Branch(Condition=CheckForBackbling())
  ]}
]
variables [
 MoveBackpackBack {type=boolean}
 MoveBackpackForward {type=boolean}
]
